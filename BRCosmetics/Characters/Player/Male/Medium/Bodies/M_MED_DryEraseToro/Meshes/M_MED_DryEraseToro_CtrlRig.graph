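# Control Rig for the DryEraseToro medium male body cosmetic, adding procedural jacket, collar and pants motion after a shared DeformRig corrective function.
# Spherical pose readers on spine, thighs, clavicles and upper arms weight additive jacket and collar bone offsets, many smoothed by AlphaInterp.
# SprintJump, Skydiving and Parachuting variables blend further jacket offsets.
# Item Harmonics then wave the pants bones, scaled by PantsHarmonicsSpeed; some pose-reader chains are left unconnected.
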ControlRigBlueprint /BRCosmetics/Characters/Player/Male/Medium/Bodies/M_MED_DryEraseToro/Meshes/M_MED_DryEraseToro_CtrlRig
graphs [
 New Function_ContainedGraph {nodes [
   n1 = RigVMFunctionEntryNode()
   RigVMFunction_Sequence = &RigVMFunction_Sequence(ExecuteContext=@n1.Argument)
   n2 = &RigVMFunction_Sequence(ExecuteContext=@RigVMFunction_Sequence.A)
   n3 = RigVMVariableNode() {variable=PantsHarmonicsSpeed}
   Clamp = &RigVMFunction_MathDoubleClamp(Value=@n3.Value)
   n4 = /Script/RigVM.RigVMFunction_MathVectorMul(B=(X=1,Y=1,Z=1), A=RigVMVariableNode() {variable=ItemHarmonics_WaveSpeed}.Value, B.X=@n3.Value, B.Y=@n3.Value, B.Z=@n3.Value) {is=struct RigVMUnitNode, template="Multiply::Execute(in A,in B,out Result)"}
   /Script/ControlRig.RigUnit_ItemHarmonics(
    RotationOrder=YZX
    Targets=((Item=(Type=Bone,Name="dyn_pants_fr_l_MSTR"),Ratio=20))
    WaveAmplitude=(X=0,Y=20,Z=15)
    WaveEase=Linear
    WaveFrequency=(X=0,Y=-8,Z=12)
    WaveMinimum=0
    WaveNoise=(X=0,Y=1.5,Z=3)
    WaveOffset=(X=0,Y=1,Z=3)
    ExecutePin=@n2.A
    WaveMaximum=@Clamp.Result
    WaveSpeed=@n4.Result
   ) {is=struct RigVMUnitNode}
   > &RigUnit_ItemHarmonics(
    RotationOrder=YZX
    Targets=((Item=(Type=Bone,Name="dyn_pants_in_l_MSTR"),Ratio=20))
    WaveAmplitude=(X=0,Y=15,Z=20)
    WaveEase=Linear
    WaveFrequency=(X=0,Y=8,Z=12)
    WaveMinimum=0
    WaveNoise=(X=0,Y=3,Z=1.5)
    WaveOffset=(X=0,Y=20,Z=1.5)
    WaveMaximum=@Clamp.Result
    WaveSpeed=@n4.Result
   )
   > &RigUnit_ItemHarmonics(
    RotationOrder=YZX
    Targets=((Item=(Type=Bone,Name="dyn_pants_rear_l_MSTR"),Ratio=20))
    WaveAmplitude=(X=0,Y=15,Z=15)
    WaveEase=Linear
    WaveFrequency=(X=0,Y=-8,Z=12)
    WaveMinimum=0
    WaveNoise=(X=0,Y=1.5,Z=3)
    WaveOffset=(X=0,Y=1,Z=3)
    WaveMaximum=@Clamp.Result
    WaveSpeed=@n4.Result
   )
   > &RigUnit_ItemHarmonics(
    RotationOrder=YZX
    Targets=((Item=(Type=Bone,Name="dyn_pants_out_l_MSTR"),Ratio=20))
    WaveAmplitude=(X=0,Y=17,Z=15)
    WaveEase=Linear
    WaveFrequency=(X=0,Y=8,Z=12)
    WaveMinimum=0
    WaveNoise=(X=0,Y=3,Z=1.5)
    WaveOffset=(X=0,Y=20,Z=1.5)
    WaveMaximum=@Clamp.Result
    WaveSpeed=@n4.Result
   )
   &RigUnit_ItemHarmonics(
    RotationOrder=YZX
    Targets=((Item=(Type=Bone,Name="dyn_pants_front_r_MSTR"),Ratio=20))
    WaveAmplitude=(X=0,Y=20,Z=15)
    WaveEase=Linear
    WaveFrequency=(X=0,Y=-8,Z=12)
    WaveMinimum=0
    WaveNoise=(X=0,Y=3,Z=1.5)
    WaveOffset=(X=0,Y=1,Z=3)
    ExecutePin=@n2.B
    WaveMaximum=@Clamp.Result
    WaveSpeed=@n4.Result
   )
   > &RigUnit_ItemHarmonics(
    RotationOrder=YZX
    Targets=((Item=(Type=Bone,Name="dyn_pants_out_r_MSTR"),Ratio=20))
    WaveAmplitude=(X=0,Y=15,Z=20)
    WaveEase=Linear
    WaveFrequency=(X=0,Y=8,Z=12)
    WaveMinimum=0
    WaveNoise=(X=0,Y=1.5,Z=3)
    WaveOffset=(X=0,Y=20,Z=1.5)
    WaveMaximum=@Clamp.Result
    WaveSpeed=@n4.Result
   )
   > &RigUnit_ItemHarmonics(
    RotationOrder=YZX
    Targets=((Item=(Type=Bone,Name="dyn_pants_rear_r_MSTR"),Ratio=20))
    WaveAmplitude=(X=0,Y=15,Z=15)
    WaveEase=Linear
    WaveFrequency=(X=0,Y=-8,Z=12)
    WaveMinimum=0
    WaveNoise=(X=0,Y=3,Z=1.5)
    WaveOffset=(X=0,Y=1,Z=3)
    WaveMaximum=@Clamp.Result
    WaveSpeed=@n4.Result
   )
   > &RigUnit_ItemHarmonics(
    RotationOrder=YZX
    Targets=((Item=(Type=Bone,Name="dyn_pants_in_r_MSTR"),Ratio=20))
    WaveAmplitude=(X=0,Y=17,Z=15)
    WaveEase=Linear
    WaveFrequency=(X=0,Y=8,Z=12)
    WaveMinimum=0
    WaveNoise=(X=0,Y=1.5,Z=3)
    WaveOffset=(X=0,Y=20,Z=1.5)
    WaveMaximum=@Clamp.Result
    WaveSpeed=@n4.Result
   )
   RigVMFunctionReturnNode(Argument=@RigVMFunction_Sequence.B)
  ]}
 New Function_ContainedGraph {nodes [
   n1 = RigVMFunctionEntryNode()
   RigVMFunction_Sequence = RigVMAggregateNode(ExecuteContext=@n1.Argument)
   RigVMFunction_Sequence_2 = RigVMAggregateNode(ExecuteContext=@RigVMFunction_Sequence.A)
   n2 = &RigUnit_SphericalPoseReader(
    ActiveRegionScaleFactors=(PositiveWidth=3,NegativeWidth=1,PositiveHeight=3,NegativeHeight=3)
    ActiveRegionSize=0.05
    Debug=(bDrawDebug=false,bDraw2D=False,bDrawLocalAxes=False,DebugScale=25,DebugSegments=20,DebugThickness=0.25)
    DriverAxis=(X=0,Y=1,Z=0)
    DriverItem=(Type=Bone,Name="spine_05")
    FalloffRegionScaleFactors=(PositiveWidth=1,NegativeWidth=1,PositiveHeight=1,NegativeHeight=1)
    FalloffSize=0.3
    FlipHeightScaling=False
    FlipWidthScaling=False
    OptionalParentItem=(Type=Bone,Name="pelvis")
    RotationOffset=(X=90,Y=0,Z=-30)
    ExecutePin=@RigVMFunction_Sequence_2.C
   )
   > &RigUnit_ModifyTransforms(
    ItemToModify="((Item=(Type=Bone,Name=\"dyn_jacket_rear_l_MSTR\"),Transform=(Rotation=(X=0,Y=-0.1736,Z=0,W=0.9848),Translation=(X=0,Y=0,Z=0),Scale3D=(X=1,Y=1,Z=1))),(Item=(Type=Bone,Name=\"dyn_jacket_rear_MSTR\"),Transform=(Rotation=(X=0,Y=-0.1736,Z=0,W=0.9848),Translation=( ...(truncated)"
    WeightMinimum=0
    Weight=@n2.OutputParam
   )
   > n3 = &RigUnit_SphericalPoseReader(
    ActiveRegionScaleFactors=(PositiveWidth=3,NegativeWidth=3,PositiveHeight=3,NegativeHeight=3)
    ActiveRegionSize=0.05
    Debug=(bDrawDebug=false,bDraw2D=False,bDrawLocalAxes=False,DebugScale=25,DebugSegments=20,DebugThickness=0.25)
    DriverAxis=(X=0,Y=1,Z=0)
    DriverItem=(Type=Bone,Name="spine_04")
    FalloffRegionScaleFactors=(PositiveWidth=1,NegativeWidth=1,PositiveHeight=1,NegativeHeight=1)
    FalloffSize=0.3
    FlipHeightScaling=False
    FlipWidthScaling=False
    OptionalParentItem=(Type=Bone,Name="spine_01")
    RotationOffset=(X=150,Y=0,Z=0)
   )
   n4 = &RigVMFunction_AlphaInterp(InterpSpeedDecreasing=10, InterpSpeedIncreasing=10, Scale=1, bClampResult=False, bInterpResult=True, Value=@n3.OutputParam)
   > &RigUnit_ModifyTransforms(
    ItemToModify="((Item=(Type=Bone,Name=\"dyn_jacket_rear_l_MSTR\"),Transform=(Rotation=(X=0,Y=-0.1305,Z=0,W=0.9914),Translation=(X=0,Y=0,Z=1),Scale3D=(X=1,Y=1,Z=1))),(Item=(Type=Bone,Name=\"dyn_jacket_rear_MSTR\"),Transform=(Rotation=(X=0,Y=-0.1305,Z=0,W=0.9914),Translation=( ...(truncated)"
    WeightMinimum=0
    Weight=@n4.Result
   )
   > n5 = &RigUnit_SphericalPoseReader(
    ActiveRegionScaleFactors=(PositiveWidth=2,NegativeWidth=2,PositiveHeight=2,NegativeHeight=2)
    ActiveRegionSize=0.05
    Debug=(bDrawDebug=false,bDraw2D=False,bDrawLocalAxes=False,DebugScale=25,DebugSegments=20,DebugThickness=0.25)
    DriverAxis=(X=0,Y=1,Z=0)
    DriverItem=(Type=Bone,Name="spine_04")
    FalloffRegionScaleFactors=(PositiveWidth=1,NegativeWidth=1,PositiveHeight=1,NegativeHeight=1)
    FalloffSize=0.25
    FlipHeightScaling=False
    FlipWidthScaling=False
    OptionalParentItem=(Type=Bone,Name="spine_01")
    RotationOffset=(X=40,Y=0,Z=0)
   )
   n6 = &RigVMFunction_AlphaInterp(InterpSpeedDecreasing=10, InterpSpeedIncreasing=10, Scale=1, bClampResult=False, bInterpResult=True, Value=@n5.OutputParam)
   > &RigUnit_ModifyTransforms(
    ItemToModify="((Item=(Type=Bone,Name=\"dyn_jacket_rear_l_MSTR\"),Transform=(Rotation=(X=0,Y=-0.1305,Z=0,W=0.9914),Translation=(X=0,Y=0,Z=-1),Scale3D=(X=1,Y=1,Z=1))),(Item=(Type=Bone,Name=\"dyn_jacket_rear_MSTR\"),Transform=(Rotation=(X=0,Y=-0.1305,Z=0,W=0.9914),Translation= ...(truncated)"
    WeightMinimum=0
    Weight=@n6.Result
   )
   n7 = &RigUnit_SphericalPoseReader(
    ActiveRegionScaleFactors=(PositiveWidth=3,NegativeWidth=1,PositiveHeight=1.2,NegativeHeight=1)
    ActiveRegionSize=0.1
    Debug=(bDrawDebug=false,bDraw2D=False,bDrawLocalAxes=False,DebugScale=25,DebugSegments=20,DebugThickness=0.25)
    DriverAxis=(X=1,Y=0,Z=0)
    DriverItem=(Type=Bone,Name="thigh_r")
    FalloffRegionScaleFactors=(PositiveWidth=1,NegativeWidth=1,PositiveHeight=0.75,NegativeHeight=0.75)
    FalloffSize=0.5
    FlipHeightScaling=true
    FlipWidthScaling=true
    OptionalParentItem=(Type=Bone,Name="None")
    RotationOffset=(X=-90,Y=0,Z=-15)
    ExecutePin=@RigVMFunction_Sequence_2.B
   )
   n8 = &RigVMFunction_AlphaInterp(InterpSpeedDecreasing=10, InterpSpeedIncreasing=10, Scale=1, bClampResult=False, bInterpResult=True, Value=@n7.OutputParam)
   > &RigUnit_ModifyTransforms(
    ItemToModify=((Item=(Type=Bone,Name="dyn_jacket_front_r_MSTR"),Transform=(Rotation=(X=-0.2082,Y=-0.0179,Z=0.0426,W=-0.977),Translation=(X=0,Y=0,Z=0),Scale3D=(X=1,Y=1,Z=1))))
    WeightMinimum=0
    Weight=@n8.Result
   )
   > n9 = &RigUnit_SphericalPoseReader(
    ActiveRegionScaleFactors=(PositiveWidth=3,NegativeWidth=1,PositiveHeight=1.3,NegativeHeight=1.2)
    ActiveRegionSize=0.1
    Debug=(bDrawDebug=false,bDraw2D=False,bDrawLocalAxes=False,DebugScale=25,DebugSegments=20,DebugThickness=0.25)
    DriverAxis=(X=1,Y=0,Z=0)
    DriverItem=(Type=Bone,Name="thigh_r")
    FalloffRegionScaleFactors=(PositiveWidth=1,NegativeWidth=1,PositiveHeight=0.5,NegativeHeight=0.5)
    FalloffSize=0.5
    FlipHeightScaling=true
    FlipWidthScaling=true
    OptionalParentItem=(Type=Bone,Name="None")
    RotationOffset=(X=-50,Y=5,Z=-12)
   )
   n10 = &RigVMFunction_AlphaInterp(InterpSpeedDecreasing=10, InterpSpeedIncreasing=10, Scale=1, bClampResult=False, bInterpResult=True, Value=@n9.OutputParam)
   > &RigUnit_ModifyTransforms(
    ItemToModify="((Item=(Type=Bone,Name=\"dyn_jacket_front_r_MSTR\"),Transform=(Rotation=(X=-0.1489,Y=-0.0027,Z=0.0627,W=-0.9869),Translation=(X=0,Y=0,Z=0),Scale3D=(X=1,Y=1,Z=1))),(Item=(Type=Bone,Name=\"dyn_jacket_side_r_MSTR\"),Transform=(Rotation=(X=0.0018,Y=0.0423,Z=-0.189 ...(truncated)"
    WeightMinimum=0
    Weight=@n10.Result
   )
   > n11 = &RigUnit_SphericalPoseReader(
    ActiveRegionScaleFactors=(PositiveWidth=3,NegativeWidth=1,PositiveHeight=1,NegativeHeight=1)
    ActiveRegionSize=0.1
    Debug=(bDrawDebug=false,bDraw2D=False,bDrawLocalAxes=False,DebugScale=25,DebugSegments=20,DebugThickness=0.25)
    DriverAxis=(X=1,Y=0,Z=0)
    DriverItem=(Type=Bone,Name="thigh_r")
    FalloffRegionScaleFactors=(PositiveWidth=1,NegativeWidth=1,PositiveHeight=1,NegativeHeight=1)
    FalloffSize=0.5
    FlipHeightScaling=true
    FlipWidthScaling=true
    OptionalParentItem=(Type=Bone,Name="None")
    RotationOffset=(X=0,Y=0,Z=0)
   )
   n12 = &RigVMFunction_AlphaInterp(InterpSpeedDecreasing=10, InterpSpeedIncreasing=10, Scale=1, bClampResult=False, bInterpResult=True, Value=@n11.OutputParam)
   > &RigUnit_ModifyTransforms(
    ItemToModify="((Item=(Type=Bone,Name=\"dyn_jacket_front_r_MSTR\"),Transform=(Rotation=(X=-0.3879,Y=0.0563,Z=0.2963,W=-0.871),Translation=(X=0,Y=0,Z=0),Scale3D=(X=1,Y=1,Z=1))),(Item=(Type=Bone,Name=\"dyn_jacket_side_r_MSTR\"),Transform=(Rotation=(X=0.3342,Y=-0.2024,Z=-0.072, ...(truncated)"
    WeightMinimum=0
    Weight=@n12.Result
   )
   n13 = &RigUnit_SphericalPoseReader(
    ActiveRegionScaleFactors=(PositiveWidth=3,NegativeWidth=1,PositiveHeight=1.2,NegativeHeight=1)
    ActiveRegionSize=0.1
    Debug=(bDrawDebug=false,bDraw2D=False,bDrawLocalAxes=False,DebugScale=25,DebugSegments=20,DebugThickness=0.25)
    DriverAxis=(X=-1,Y=0,Z=0)
    DriverItem=(Type=Bone,Name="thigh_l")
    FalloffRegionScaleFactors=(PositiveWidth=1,NegativeWidth=1,PositiveHeight=0.75,NegativeHeight=0.75)
    FalloffSize=0.5
    FlipHeightScaling=False
    FlipWidthScaling=False
    OptionalParentItem=(Type=Bone,Name="None")
    RotationOffset=(X=90,Y=0,Z=-15)
    ExecutePin=@RigVMFunction_Sequence_2.A
   )
   RigVMFunction_AlphaInterp = &RigVMFunction_AlphaInterp(InterpSpeedDecreasing=10, InterpSpeedIncreasing=10, Scale=1, bClampResult=False, bInterpResult=True, Value=@n13.OutputParam)
   > &RigUnit_ModifyTransforms(
    ItemToModify="((Item=(Type=Bone,Name=\"dyn_jacket_front_l_MSTR\"),Transform=(Rotation=(X=0.2734,Y=-0.0239,Z=-0.0838,W=0.958),Translation=(X=0,Y=0,Z=0),Scale3D=(X=1,Y=1,Z=1))),(Item=(Type=Bone,Name=\"dyn_jacket_side_l_MSTR\"),Transform=(Rotation=(X=0.1012,Y=-0.0089,Z=-0.0867 ...(truncated)"
    WeightMinimum=0
    Weight=@RigVMFunction_AlphaInterp.Result
   )
   > n14 = &RigUnit_SphericalPoseReader(
    ActiveRegionScaleFactors=(PositiveWidth=3,NegativeWidth=1,PositiveHeight=1.3,NegativeHeight=1.2)
    ActiveRegionSize=0.1
    Debug=(bDrawDebug=false,bDraw2D=False,bDrawLocalAxes=False,DebugScale=25,DebugSegments=20,DebugThickness=0.25)
    DriverAxis=(X=-1,Y=0,Z=0)
    DriverItem=(Type=Bone,Name="thigh_l")
    FalloffRegionScaleFactors=(PositiveWidth=1,NegativeWidth=1,PositiveHeight=0.5,NegativeHeight=0.5)
    FalloffSize=0.5
    FlipHeightScaling=False
    FlipWidthScaling=False
    OptionalParentItem=(Type=Bone,Name="None")
    RotationOffset=(X=130,Y=5,Z=-12)
   )
   n15 = &RigVMFunction_AlphaInterp(InterpSpeedDecreasing=10, InterpSpeedIncreasing=10, Scale=1, bClampResult=False, bInterpResult=True, Value=@n14.OutputParam)
   > &RigUnit_ModifyTransforms(
    ItemToModify="((Item=(Type=Bone,Name=\"dyn_jacket_front_l_MSTR\"),Transform=(Rotation=(X=0.0109,Y=0.0269,Z=-0.139,W=0.9899),Translation=(X=0,Y=0,Z=0),Scale3D=(X=1,Y=1,Z=1))),(Item=(Type=Bone,Name=\"dyn_jacket_side_l_MSTR\"),Transform=(Rotation=(X=0,Y=0,Z=-0.1547,W=0.988),Tr ...(truncated)"
    WeightMinimum=0
    Weight=@n15.Result
   )
   > n16 = &RigUnit_SphericalPoseReader(
    ActiveRegionScaleFactors=(PositiveWidth=3,NegativeWidth=1,PositiveHeight=1,NegativeHeight=1)
    ActiveRegionSize=0.1
    Debug=(bDrawDebug=false,bDraw2D=False,bDrawLocalAxes=False,DebugScale=25,DebugSegments=20,DebugThickness=0.25)
    DriverAxis=(X=-1,Y=0,Z=0)
    DriverItem=(Type=Bone,Name="thigh_l")
    FalloffRegionScaleFactors=(PositiveWidth=1,NegativeWidth=1,PositiveHeight=1,NegativeHeight=1)
    FalloffSize=0.5
    FlipHeightScaling=False
    FlipWidthScaling=False
    OptionalParentItem=(Type=Bone,Name="None")
    RotationOffset=(X=180,Y=0,Z=0)
   )
   n17 = &RigVMFunction_AlphaInterp(InterpSpeedDecreasing=10, InterpSpeedIncreasing=10, Scale=1, bClampResult=False, bInterpResult=True, Value=@n16.OutputParam)
   > &RigUnit_ModifyTransforms(
    ItemToModify="((Item=(Type=Bone,Name=\"dyn_jacket_front_l_MSTR\"),Transform=(Rotation=(X=0.5716,Y=-0.0841,Z=-0.2007,W=0.7912),Translation=(X=0,Y=0,Z=0),Scale3D=(X=1,Y=1,Z=1))),(Item=(Type=Bone,Name=\"dyn_jacket_side_l_MSTR\"),Transform=(Rotation=(X=0.5251,Y=-0.161,Z=-0.0286 ...(truncated)"
    WeightMinimum=0
    Weight=@n17.Result
   )
   n18 = &RigVMFunction_Sequence(ExecuteContext=@RigVMFunction_Sequence.B)
   n19 = &RigUnit_SphericalPoseReader(
    ActiveRegionScaleFactors=(PositiveWidth=1,NegativeWidth=1,PositiveHeight=1,NegativeHeight=1)
    ActiveRegionSize=0.1
    Debug=(bDrawDebug=false,bDraw2D=False,bDrawLocalAxes=False,DebugScale=25,DebugSegments=20,DebugThickness=0.25)
    DriverAxis=(X=1,Y=0,Z=0)
    DriverItem=(Type=Bone,Name="upperarm_l")
    FalloffRegionScaleFactors=(PositiveWidth=1,NegativeWidth=1,PositiveHeight=1,NegativeHeight=1)
    FalloffSize=0.3
    FlipHeightScaling=False
    FlipWidthScaling=False
    OptionalParentItem=(Type=Bone,Name="spine_05")
    RotationOffset=(X=50,Y=0,Z=20)
    ExecutePin=@n18.A
   )
   > &RigUnit_ModifyTransforms(
    ItemToModify=((Item=(Type=Bone,Name="dyn_collar_l_MSTR"),Transform=(Rotation=(X=0,Y=-0.4226,Z=0,W=0.9063),Translation=(X=0,Y=0,Z=0),Scale3D=(X=1,Y=1,Z=1))))
    WeightMinimum=0
    Weight=@n19.OutputParam
   )
   > n20 = &RigUnit_SphericalPoseReader(
    ActiveRegionScaleFactors=(PositiveWidth=1,NegativeWidth=1,PositiveHeight=1,NegativeHeight=1)
    ActiveRegionSize=0.1
    Debug=(bDrawDebug=false,bDraw2D=False,bDrawLocalAxes=False,DebugScale=25,DebugSegments=20,DebugThickness=0.25)
    DriverAxis=(X=1,Y=0,Z=0)
    DriverItem=(Type=Bone,Name="upperarm_l")
    FalloffRegionScaleFactors=(PositiveWidth=1,NegativeWidth=1,PositiveHeight=1,NegativeHeight=1)
    FalloffSize=0.3
    FlipHeightScaling=False
    FlipWidthScaling=False
    OptionalParentItem=(Type=Bone,Name="spine_05")
    RotationOffset=(X=10,Y=30,Z=0)
   )
   > &RigUnit_ModifyTransforms(
    ItemToModify=((Item=(Type=Bone,Name="dyn_collar_l_MSTR"),Transform=(Rotation=(X=0,Y=-0.0872,Z=0,W=0.9962),Translation=(X=0,Y=0,Z=0),Scale3D=(X=1,Y=1,Z=1))))
    WeightMinimum=0
    Weight=@n20.OutputParam
   )
   > n21 = &RigUnit_SphericalPoseReader(
    ActiveRegionScaleFactors=(PositiveWidth=1,NegativeWidth=1,PositiveHeight=1,NegativeHeight=1)
    ActiveRegionSize=0.15
    Debug=(bDrawDebug=false,bDraw2D=False,bDrawLocalAxes=False,DebugScale=25,DebugSegments=20,DebugThickness=0.25)
    DriverAxis=(X=1,Y=0,Z=0)
    DriverItem=(Type=Bone,Name="clavicle_l")
    FalloffRegionScaleFactors=(PositiveWidth=1,NegativeWidth=1,PositiveHeight=1,NegativeHeight=1)
    FalloffSize=0.3
    FlipHeightScaling=False
    FlipWidthScaling=False
    OptionalParentItem=(Type=Bone,Name="spine_05")
    RotationOffset=(X=-30,Y=0,Z=90)
   )
   > &RigUnit_ModifyTransforms(
    ItemToModify=((Item=(Type=Bone,Name="dyn_collar_l_MSTR"),Transform=(Rotation=(X=-0.122,Y=-0.3268,Z=-0.1613,W=0.9232),Translation=(X=-0.3815,Y=0.0646,Z=0.0673),Scale3D=(X=1,Y=1,Z=1))))
    WeightMinimum=0
    Weight=@n21.OutputParam
   )
   n22 = &RigUnit_SphericalPoseReader(
    ActiveRegionScaleFactors=(PositiveWidth=1,NegativeWidth=1,PositiveHeight=1,NegativeHeight=1)
    ActiveRegionSize=0.1
    Debug=(bDrawDebug=false,bDraw2D=False,bDrawLocalAxes=False,DebugScale=25,DebugSegments=20,DebugThickness=0.25)
    DriverAxis=(X=-1,Y=0,Z=0)
    DriverItem=(Type=Bone,Name="upperarm_r")
    FalloffRegionScaleFactors=(PositiveWidth=1,NegativeWidth=1,PositiveHeight=1,NegativeHeight=1)
    FalloffSize=0.3
    FlipHeightScaling=False
    FlipWidthScaling=False
    OptionalParentItem=(Type=Bone,Name="spine_05")
    RotationOffset=(X=-100,Y=0,Z=50)
    ExecutePin=@n18.B
   )
   > &RigUnit_ModifyTransforms(
    ItemToModify=((Item=(Type=Bone,Name="dyn_collar_r_MSTR"),Transform=(Rotation=(X=0,Y=-0.766,Z=0,W=0.6428),Translation=(X=0,Y=0,Z=0),Scale3D=(X=1,Y=1,Z=1))))
    WeightMinimum=0
    Weight=@n22.OutputParam
   )
   > n23 = &RigUnit_SphericalPoseReader(
    ActiveRegionScaleFactors=(PositiveWidth=1,NegativeWidth=1,PositiveHeight=1,NegativeHeight=1)
    ActiveRegionSize=0.15
    Debug=(bDrawDebug=false,bDraw2D=False,bDrawLocalAxes=False,DebugScale=25,DebugSegments=20,DebugThickness=0.25)
    DriverAxis=(X=-1,Y=0,Z=0)
    DriverItem=(Type=Bone,Name="clavicle_r")
    FalloffRegionScaleFactors=(PositiveWidth=1,NegativeWidth=1,PositiveHeight=1,NegativeHeight=1)
    FalloffSize=0.3
    FlipHeightScaling=False
    FlipWidthScaling=False
    OptionalParentItem=(Type=Bone,Name="spine_05")
    RotationOffset=(X=150,Y=0,Z=90)
   )
   > &RigUnit_ModifyTransforms(
    ItemToModify=((Item=(Type=Bone,Name="dyn_collar_r_MSTR"),Transform=(Rotation=(X=-0.1496,Y=-0.4824,Z=0.0421,W=0.862),Translation=(X=0,Y=0,Z=0),Scale3D=(X=1,Y=1,Z=1))))
    WeightMinimum=0
    Weight=@n23.OutputParam
   )
   n24 = RigVMVariableNode() {variable=SprintJump}
   &RigUnit_ModifyTransforms(
    ItemToModify="((Item=(Type=Bone,Name=\"dyn_jacket_rear_l_MSTR\"),Transform=(Rotation=(X=0,Y=-0.0872,Z=0,W=0.9962),Translation=(X=0,Y=0,Z=1.5),Scale3D=(X=1,Y=1,Z=1))),(Item=(Type=Bone,Name=\"dyn_jacket_rear_MSTR\"),Transform=(Rotation=(X=0,Y=-0.0872,Z=0,W=0.9962),Translation ...(truncated)"
    WeightMinimum=0
    Weight=@n24.Value
    ExecutePin=@RigVMFunction_Sequence.D
   )
   n25 = &RigVMFunction_AlphaInterp(InterpSpeedDecreasing=10, InterpSpeedIncreasing=10, Scale=1, bClampResult=true, Value=RigVMVariableNode() {variable=Skydiving}.Value)
   > &RigUnit_ModifyTransforms(
    ItemToModify="((Item=(Type=Bone,Name=\"dyn_jacket_front_l_MSTR\"),Transform=(Rotation=(X=0,Y=0,Z=0,W=1),Translation=(X=0,Y=0,Z=0.5),Scale3D=(X=1,Y=1,Z=1))),(Item=(Type=Bone,Name=\"dyn_jacket_side_l_MSTR\"),Transform=(Rotation=(X=0,Y=0,Z=0,W=1),Translation=(X=0,Y=0,Z=0.5),Sc ...(truncated)"
    WeightMinimum=0
    Weight=@n25.Result
   )
   AlphaInterp = &RigVMFunction_AlphaInterp(InterpSpeedDecreasing=60, InterpSpeedIncreasing=60, Scale=1, bClampResult=true, Value=RigVMVariableNode() {variable=Parachuting}.Value)
   > &RigUnit_ModifyTransforms(
    ItemToModify="((Item=(Type=Bone,Name=\"dyn_jacket_front_l_MSTR\"),Transform=(Rotation=(X=0,Y=-0.0436,Z=0,W=0.999),Translation=(X=0,Y=0,Z=1.5),Scale3D=(X=1,Y=1,Z=1))),(Item=(Type=Bone,Name=\"dyn_jacket_side_l_MSTR\"),Transform=(Rotation=(X=0,Y=-0.0436,Z=0,W=0.999),Translatio ...(truncated)"
    WeightMinimum=0
    Weight=@AlphaInterp.Result
   )
   &RigUnit_SphericalPoseReader(
    ActiveRegionScaleFactors=(PositiveWidth=3,NegativeWidth=1,PositiveHeight=1.5,NegativeHeight=1.5)
    ActiveRegionSize=0.1
    Debug=(bDrawDebug=false,bDraw2D=False,bDrawLocalAxes=False,DebugScale=25,DebugSegments=20,DebugThickness=0.25)
    DriverAxis=(X=-1,Y=0,Z=0)
    DriverItem=(Type=Bone,Name="thigh_l")
    ExecutePin=()
    FalloffRegionScaleFactors=(PositiveWidth=1,NegativeWidth=1,PositiveHeight=1,NegativeHeight=1)
    FalloffSize=0.5
    FlipHeightScaling=False
    FlipWidthScaling=False
    OptionalParentItem=(Type=Bone,Name="None")
    RotationOffset=(X=-125,Y=0,Z=0)
   )
   > &RigUnit_SphericalPoseReader(
    ActiveRegionScaleFactors=(PositiveWidth=3,NegativeWidth=1,PositiveHeight=3,NegativeHeight=3)
    ActiveRegionSize=0.1
    Debug=(bDrawDebug=false,bDraw2D=False,bDrawLocalAxes=False,DebugScale=25,DebugSegments=20,DebugThickness=0.25)
    DriverAxis=(X=-1,Y=0,Z=0)
    DriverItem=(Type=Bone,Name="thigh_l")
    FalloffRegionScaleFactors=(PositiveWidth=1,NegativeWidth=0.95,PositiveHeight=1,NegativeHeight=1)
    FalloffSize=0.5
    FlipHeightScaling=False
    FlipWidthScaling=False
    OptionalParentItem=(Type=Bone,Name="None")
    RotationOffset=(X=-90,Y=0,Z=0)
   )
   n26 = &RigUnit_SphericalPoseReader(
    ActiveRegionScaleFactors=(PositiveWidth=3,NegativeWidth=1,PositiveHeight=1.5,NegativeHeight=1.5)
    ActiveRegionSize=0.1
    Debug=(bDrawDebug=false,bDraw2D=False,bDrawLocalAxes=False,DebugScale=25,DebugSegments=20,DebugThickness=0.25)
    DriverAxis=(X=1,Y=0,Z=0)
    DriverItem=(Type=Bone,Name="thigh_r")
    ExecutePin=()
    FalloffRegionScaleFactors=(PositiveWidth=1,NegativeWidth=1,PositiveHeight=1,NegativeHeight=1)
    FalloffSize=0.5
    FlipHeightScaling=true
    FlipWidthScaling=true
    OptionalParentItem=(Type=Bone,Name="None")
    RotationOffset=(X=55,Y=0,Z=0)
   )
   n27 = &RigVMFunction_AlphaInterp(InterpSpeedDecreasing=10, InterpSpeedIncreasing=10, Scale=1, bClampResult=False, bInterpResult=True, Value=@n26.OutputParam)
   > &RigUnit_ModifyTransforms(
    ItemToModify="((Item=(Type=Bone,Name=\"dyn_jacket_rear_r_MSTR\"),Transform=(Rotation=(X=0,Y=-0.2887,Z=0,W=0.9574),Translation=(X=0,Y=0,Z=0),Scale3D=(X=1,Y=1,Z=1))),(Item=(Type=Bone,Name=\"dyn_jacket_side_r_MSTR\"),Transform=(Rotation=(X=0,Y=0.2659,Z=0,W=-0.964),Translation= ...(truncated)"
    WeightMinimum=0
    Weight=@n27.Result
   )
   > n28 = &RigUnit_SphericalPoseReader(
    ActiveRegionScaleFactors=(PositiveWidth=3,NegativeWidth=1,PositiveHeight=3,NegativeHeight=3)
    ActiveRegionSize=0.1
    Debug=(bDrawDebug=false,bDraw2D=False,bDrawLocalAxes=False,DebugScale=25,DebugSegments=20,DebugThickness=0.25)
    DriverAxis=(X=1,Y=0,Z=0)
    DriverItem=(Type=Bone,Name="thigh_r")
    FalloffRegionScaleFactors=(PositiveWidth=1,NegativeWidth=0.95,PositiveHeight=1,NegativeHeight=1)
    FalloffSize=0.5
    FlipHeightScaling=true
    FlipWidthScaling=true
    OptionalParentItem=(Type=Bone,Name="None")
    RotationOffset=(X=90,Y=0,Z=0)
   )
   n29 = &RigVMFunction_AlphaInterp(InterpSpeedDecreasing=10, InterpSpeedIncreasing=10, Scale=1, bClampResult=False, bInterpResult=True, Value=@n28.OutputParam)
   > &RigUnit_ModifyTransforms(
    ItemToModify="((Item=(Type=Bone,Name=\"dyn_jacket_rear_r_MSTR\"),Transform=(Rotation=(X=0,Y=-0.0872,Z=0,W=0.9962),Translation=(X=0,Y=0,Z=0),Scale3D=(X=1,Y=1,Z=1))),(Item=(Type=Bone,Name=\"dyn_jacket_rear_MSTR\"),Transform=(Rotation=(X=0,Y=-0.2659,Z=0,W=0.964),Translation=(X ...(truncated)"
    WeightMinimum=0
    Weight=@n29.Result
   )
   > n30 = &RigUnit_SphericalPoseReader(
    ActiveRegionScaleFactors=(PositiveWidth=3,NegativeWidth=1,PositiveHeight=1.2,NegativeHeight=1)
    ActiveRegionSize=0.1
    Debug=(bDrawDebug=true,bDraw2D=False,bDrawLocalAxes=False,DebugScale=25,DebugSegments=20,DebugThickness=0.25)
    DriverAxis=(X=1,Y=0,Z=0)
    DriverItem=(Type=Bone,Name="thigh_r")
    FalloffRegionScaleFactors=(PositiveWidth=1,NegativeWidth=1,PositiveHeight=0.3,NegativeHeight=0.3)
    FalloffSize=0.58
    FlipHeightScaling=true
    FlipWidthScaling=true
    OptionalParentItem=(Type=Bone,Name="None")
    RotationOffset=(X=-135,Y=0,Z=-15)
   )
   n31 = &RigVMFunction_AlphaInterp(InterpSpeedDecreasing=10, InterpSpeedIncreasing=10, Scale=1, bClampResult=False, bInterpResult=True, Value=@n30.OutputParam)
   > &RigUnit_ModifyTransforms(
    ItemToModify=((Item=(Type=Bone,Name="dyn_jacket_side_r_MSTR"),Transform=(Rotation=(X=0,Y=-0.1334,Z=0,W=0.9911),Translation=(X=0,Y=0,Z=0),Scale3D=(X=1,Y=1,Z=1))))
    WeightMinimum=0
    Weight=@n31.Result
   )
   n32 = &RigUnit_SphericalPoseReader(
    ActiveRegionScaleFactors=(PositiveWidth=3,NegativeWidth=1,PositiveHeight=1.2,NegativeHeight=1)
    ActiveRegionSize=0.1
    Debug=(bDrawDebug=true,bDraw2D=False,bDrawLocalAxes=False,DebugScale=25,DebugSegments=20,DebugThickness=0.25)
    DriverAxis=(X=-1,Y=0,Z=0)
    DriverItem=(Type=Bone,Name="thigh_l")
    ExecutePin=()
    FalloffRegionScaleFactors=(PositiveWidth=1,NegativeWidth=1,PositiveHeight=0.3,NegativeHeight=0.3)
    FalloffSize=0.58
    FlipHeightScaling=False
    FlipWidthScaling=False
    OptionalParentItem=(Type=Bone,Name="None")
    RotationOffset=(X=45,Y=0,Z=-15)
   )
   n33 = &RigVMFunction_AlphaInterp(InterpSpeedDecreasing=10, InterpSpeedIncreasing=10, Scale=1, bClampResult=False, bInterpResult=True, Value=@n32.OutputParam)
   > &RigUnit_ModifyTransforms(
    ItemToModify=((Item=(Type=Bone,Name="dyn_jacket_side_r_MSTR"),Transform=(Rotation=(X=0,Y=-0.1334,Z=0,W=0.9911),Translation=(X=0,Y=0,Z=0),Scale3D=(X=1,Y=1,Z=1))))
    WeightMinimum=0
    Weight=@n33.Result
   )
   n34 = &RigUnit_SphericalPoseReader(
    ActiveRegionScaleFactors=(PositiveWidth=1,NegativeWidth=1,PositiveHeight=1,NegativeHeight=1)
    ActiveRegionSize=0.075
    Debug=(bDrawDebug=true,bDraw2D=False,bDrawLocalAxes=False,DebugScale=25,DebugSegments=20,DebugThickness=0.25)
    DriverAxis=(X=0,Y=1,Z=0)
    DriverItem=(Type=Bone,Name="spine_05")
    ExecutePin=()
    FalloffRegionScaleFactors=(PositiveWidth=1,NegativeWidth=1,PositiveHeight=1,NegativeHeight=1)
    FalloffSize=0.3
    FlipHeightScaling=False
    FlipWidthScaling=False
    OptionalParentItem=(Type=Bone,Name="pelvis")
    RotationOffset=(X=80,Y=0,Z=3)
   )
   > &RigUnit_ModifyTransforms(
    ItemToModify="((Item=(Type=Bone,Name=\"dyn_jacket_front_r_MSTR\"),Transform=(Rotation=(X=0,Y=-0.0436,Z=0,W=0.999),Translation=(X=0,Y=0,Z=0),Scale3D=(X=1,Y=1,Z=1))),(Item=(Type=Bone,Name=\"dyn_jacket_side_r_MSTR\"),Transform=(Rotation=(X=0,Y=-0.0436,Z=0,W=0.999),Translation= ...(truncated)"
    WeightMinimum=0
    Weight=@n34.OutputParam
   )
   &RigVMFunction_AlphaInterp(InterpSpeedDecreasing=10, InterpSpeedIncreasing=10, Scale=1, bClampResult=true, Value=@n24.Value)
   RigVMFunctionReturnNode(Argument=@RigVMFunction_Sequence.C)
  ]}
 RigVMFunctionLibrary {nodes [
   Pants Harmonics = RigVMCollapseNode(Argument=())
   Solve Jacket and Collar = RigVMCollapseNode(Argument=())
  ]}
 RigVMFunction_Sequence_2_ContainedGraph {nodes [
   n1 = RigVMFunctionEntryNode()
   RigVMFunction_Sequence = &RigVMFunction_Sequence(ExecuteContext=@n1.ExecuteContext)
   n2 = &RigVMFunction_Sequence(ExecuteContext=@RigVMFunction_Sequence.B)
   n3 = &RigVMFunction_Sequence(ExecuteContext=@n2.B)
   RigVMFunctionReturnNode(A=@RigVMFunction_Sequence.A, B=@n2.A, C=@n3.A, D=@n3.B)
  ]}
 RigVMFunction_Sequence_3_ContainedGraph {nodes [
   n1 = RigVMFunctionEntryNode()
   n2 = &RigVMFunction_Sequence(ExecuteContext=@n1.ExecuteContext)
   n3 = &RigVMFunction_Sequence(ExecuteContext=@n2.B)
   n4 = &RigVMFunction_Sequence(ExecuteContext=@n3.B)
   n5 = &RigVMFunction_Sequence(ExecuteContext=@n4.B)
   RigVMFunctionReturnNode(B=@n3.A, A=@n2.A, C=@n4.A, D=@n5.A, E=@n5.B)
  ]}
 RigVMModel {nodes [
   RigUnit_BeginExecution = &RigUnit_BeginExecution()
   FN_DeformRig_FNC = RigVMFunctionReferenceNode(
    DeformRig_ Curve="(deformRig_l_upperarm_up_alpha_72_#1=False,deformRig_r_upperarm_up_alpha_73_#2=False,deformRig_l_upperarm_fr_alpha_74_#3=False,deformRig_r_upperarm_fr_alpha_75_CB51B8 ...(truncated)"
    DeformRig_Bool="(upperarm_1_#4=True,elbow_3_#5=True,wrist_7_#6=True,groin_8_#7=True,glute_17_#8=True,knee_9_40C408324CA261 ...(truncated)"
    DeformRig_Debug=(upperarm_debug_18_#4=False,wrist_debug_20_#6=False,lat_debug_26_#9=False,groinGlute_debug_27_#7=False)
    DeformRig_Value="(upperarm_up_18_#10=(Rotation=(X=0,Y=0,Z=0,W=1),Translation=(X=0,Y=0,Z=8),Scale3D=(X=0.7,Y=1,Z=1)),upperarm_fwd_fr_19_#11=(Rotation=(X=0,Y=0,Z=0,W=1),Translation=(X=0,Y=10,Z=0),Scale3D=(X=0.5,Y=1,Z= ...(truncated)"
    ExecuteContext=@RigUnit_BeginExecution.ExecutePin
   )
   Solve Jacket and Collar_1 = RigVMFunctionReferenceNode(Argument=@FN_DeformRig_FNC.ExecuteContext)
   Pants Harmonics = RigVMFunctionReferenceNode(Argument=@"Solve Jacket and Collar_1".Argument)
   Pants Harmonics_0 = RigVMFunctionReferenceNode(Argument=())
  ]}
]
hierarchy [
 root {loc=<0 0 0>, rot=<0 0 0 1>}
 pelvis {parent=root, loc=<0 -0.7071 93.1191>, rot=<0.0127 0.707 -0.0127 -0.707>}
 spine_01 {parent=pelvis, loc=<6.7494 -0.6388 0>, rot=<0 0 0.034 -0.9994>}
 spine_02 {parent=spine_01, loc=<7.2518 -0.035 0>, rot=<0 0 0.0238 -0.9997>}
 spine_03 {parent=spine_02, loc=<6.0104 0 0>, rot=<0 0 0.0149 -0.9999>}
 spine_04 {parent=spine_03, loc=<8.1024 0 0>, rot=<0 0 -0.0132 -0.9999>}
 spine_05 {parent=spine_04, loc=<10.0146 0 0>, rot=<0 0 -0.0015 -1>}
 neck_01 {parent=spine_05, loc=<11.3138 0.1529 0>, rot=<0 0 -0.2003 -0.9797>}
 neck_02 {parent=neck_01, loc=<4.5627 0 0>, rot=<0 0 0.0504 -0.9987>}
 head {parent=neck_02, loc=<5.3914 0 0>, rot=<0 0 0.115 -0.9934>}
 attach_cape {parent=spine_05, loc=<9.0169 -11.7094 -0.0007>, rot=<0.0616 -0.7044 -0.0616 -0.7044>}
 cape {parent=attach_cape, type=Null, loc=<0 0 0>, rot=<0 0 0 1>}
 attach_backpack {parent=spine_05, loc=<-0.1228 -13.4681 0.0001>, rot=<-0.0093 -0.707 0.0092 -0.707>}
 Backpack_BR {parent=attach_backpack, type=Null, loc=<0 0 0>, rot=<0 0 0 1>}
 clavicle_l {parent=spine_05, loc=<4.3575 -0.1466 -2.3223>, rot=<0.0465 -0.7289 -0.009 -0.6829>}
 upperarm_l {parent=clavicle_l, loc=<15.5111 0 0>, rot=<0.0226 -0.3234 -0.0088 -0.9459>}
 lowerarm_l {parent=upperarm_l, loc=<26.6514 0 0>, rot=<0.0391 -0.0099 -0.2453 -0.9686>}
 hand_l {parent=lowerarm_l, loc=<22.3491 0 0>, rot=<-0.7081 -0.0507 0.0975 -0.6975>}
 thumb_01_l {parent=hand_l, loc=<1.3118 -0.3191 -2.459>, rot=<0.6546 -0.4042 0.0664 -0.6354>}
 thumb_02_l {parent=thumb_01_l, loc=<5.9336 -0.0448 0.0401>, rot=<-0.0037 0.0213 0.298 -0.9543>}
 thumb_03_l {parent=thumb_02_l, loc=<2.2223 0.3682 -0.0907>, rot=<0.0118 -0.013 0.0066 -0.9998>}
 index_metacarpal_l {parent=hand_l, loc=<1.6535 0.0183 -2.96>, rot=<0.0158 -0.0124 -0.0173 -0.9996>}
 index_01_l {parent=index_metacarpal_l, loc=<8.3008 0.0357 -0.0524>, rot=<-0.0044 0.0485 0.0924 -0.9945>}
 index_02_l {parent=index_01_l, loc=<3.3587 -0.3004 0.0452>, rot=<-0.002 0.0027 0.3225 -0.9466>}
 index_03_l {parent=index_02_l, loc=<2.5578 -0.123 -0.0252>, rot=<-0.002 0.0067 0.2558 -0.9667>}
 middle_metacarpal_l {parent=hand_l, loc=<2.1856 -0.1725 -0.5984>, rot=<0.006 0.0168 -0.0329 -0.9993>}
 middle_01_l {parent=middle_metacarpal_l, loc=<7.3275 -0.2447 0.0129>, rot=<-0.0208 0.0042 0.0858 -0.9961>}
 middle_02_l {parent=middle_01_l, loc=<4.0123 0.1401 0.1175>, rot=<-0.0218 0.0258 0.3512 -0.9357>}
 middle_03_l {parent=middle_02_l, loc=<2.6077 0.0534 0.1373>, rot=<0.0082 0.0188 0.1708 -0.9851>}
 ring_metacarpal_l {parent=hand_l, loc=<2.2074 -0.3625 1.3638>, rot=<-0.0284 0.0722 -0.0458 -0.9959>}
 ring_01_l {parent=ring_metacarpal_l, loc=<7.0295 -0.6197 -0.1446>, rot=<-0.0489 -0.0252 0.1892 -0.9804>}
 ring_02_l {parent=ring_01_l, loc=<3.6174 0.4647 -0.1027>, rot=<0 0 0.2806 -0.9598>}
 ring_03_l {parent=ring_02_l, loc=<2.4152 0.114 0.0199>, rot=<0 0 0.1189 -0.9929>}
 pinky_metacarpal_l {parent=hand_l, loc=<1.7616 -0.9753 3.3814>, rot=<-0.0397 0.0814 -0.069 -0.9935>}
 pinky_01_l {parent=pinky_metacarpal_l, loc=<7.1709 -0.4831 -0.1094>, rot=<-0.0654 -0.0035 0.2242 -0.9723>}
 pinky_02_l {parent=pinky_01_l, loc=<2.8274 0.0893 -0.1043>, rot=<0.0037 0.0011 0.2854 -0.9584>}
 pinky_03_l {parent=pinky_02_l, loc=<2.2046 0.0307 -0.0078>, rot=<-0.0002 0.0098 0.1445 -0.9895>}
 deform_wrist_dn_base_l {parent=hand_l, loc=<0 -2.2342 0>, rot=<0 0 0 -1>}
 deform_wrist_dn_l {parent=deform_wrist_dn_base_l, loc=<0 0 0>, rot=<0 0 0 1>}
 deform_wrist_up_base_l {parent=hand_l, loc=<0 2.2682 0>, rot=<0 0 0 -1>}
 deform_wrist_up_l {parent=deform_wrist_up_base_l, loc=<0 0 0>, rot=<0 0 0 1>}
 lowerarm_twist_02_l {parent=lowerarm_l, loc=<9.6199 0 0>, rot=<0 0 0 -1>}
 lowerarm_twist_01_l {parent=lowerarm_l, loc=<17.8549 0 0>, rot=<0 0 0 -1>}
 deform_elbow_in_base_l {parent=lowerarm_l, loc=<0.2112 5.259 0.8274>, rot=<0 0 0 -1>}
 deform_elbow_in_l {parent=deform_elbow_in_base_l, loc=<0 0 0>, rot=<0 0 0 1>}
 deform_elbow_out_base_l {parent=lowerarm_l, loc=<0 -4 0>, rot=<0 0 0 -1>}
 deform_elbow_out_l {parent=deform_elbow_out_base_l, loc=<0 0 0>, rot=<0 0 0 1>}
 upperarm_twist_01_l {parent=upperarm_l, loc=<0.0041 0 0>, rot=<0 0 0 -1>}
 deform_upperarm_bk_base_l {parent=upperarm_twist_01_l, loc=<1.0813 -7.2193 -1.9453>, rot=<0 0 0 1>}
 deform_upperarm_bk_l {parent=deform_upperarm_bk_base_l, loc=<0 0 0>, rot=<0 0 0 1>}
 deform_upperarm_fr_base_l {parent=upperarm_twist_01_l, loc=<-0.3691 7.0506 -2.765>, rot=<0 0 0 1>}
 deform_upperarm_fr_l {parent=deform_upperarm_fr_base_l, loc=<0 0 0>, rot=<0 0 0 1>}
 deform_upperarm_up_base_l {parent=upperarm_twist_01_l, loc=<-3.2216 -1.1391 3.3311>, rot=<0 0 0 1>}
 deform_upperarm_up_l {parent=deform_upperarm_up_base_l, loc=<0 0 0>, rot=<0 0 0 1>}
 upperarm_twist_02_l {parent=upperarm_l, loc=<13.7941 0 0>, rot=<0 0 0 -1>}
 clavicle_r {parent=spine_05, loc=<4.3575 -0.1466 2.3223>, rot=<0.7289 0.0465 -0.6829 0.009>}
 upperarm_r {parent=clavicle_r, loc=<-15.5111 0 0>, rot=<0.0226 -0.3234 -0.0088 -0.9459>}
 lowerarm_r {parent=upperarm_r, loc=<-26.6514 0 0>, rot=<0.0391 -0.0099 -0.2453 -0.9686>}
 hand_r {parent=lowerarm_r, loc=<-22.3491 0 0>, rot=<-0.7081 -0.0507 0.0975 -0.6975>}
 thumb_01_r {parent=hand_r, loc=<-1.3118 0.3191 2.459>, rot=<0.6546 -0.4042 0.0664 -0.6354>}
 thumb_02_r {parent=thumb_01_r, loc=<-5.9336 0.0448 -0.0401>, rot=<-0.0037 0.0213 0.298 -0.9543>}
 thumb_03_r {parent=thumb_02_r, loc=<-2.2223 -0.3682 0.0907>, rot=<0.0118 -0.013 0.0066 -0.9998>}
 index_metacarpal_r {parent=hand_r, loc=<-1.6535 -0.0183 2.96>, rot=<0.0158 -0.0124 -0.0173 -0.9996>}
 index_01_r {parent=index_metacarpal_r, loc=<-8.3008 -0.0357 0.0524>, rot=<-0.0044 0.0485 0.0924 -0.9945>}
 index_02_r {parent=index_01_r, loc=<-3.3587 0.3004 -0.0452>, rot=<-0.002 0.0027 0.3225 -0.9466>}
 index_03_r {parent=index_02_r, loc=<-2.5578 0.123 0.0252>, rot=<-0.002 0.0067 0.2558 -0.9667>}
 middle_metacarpal_r {parent=hand_r, loc=<-2.1856 0.1725 0.5984>, rot=<0.006 0.0168 -0.0329 -0.9993>}
 middle_01_r {parent=middle_metacarpal_r, loc=<-7.3275 0.2447 -0.0129>, rot=<-0.0208 0.0042 0.0858 -0.9961>}
 middle_02_r {parent=middle_01_r, loc=<-4.0123 -0.1401 -0.1175>, rot=<-0.0218 0.0258 0.3512 -0.9357>}
 middle_03_r {parent=middle_02_r, loc=<-2.6077 -0.0534 -0.1373>, rot=<0.0082 0.0188 0.1708 -0.9851>}
 ring_metacarpal_r {parent=hand_r, loc=<-2.2074 0.3625 -1.3638>, rot=<-0.0284 0.0722 -0.0458 -0.9959>}
 ring_01_r {parent=ring_metacarpal_r, loc=<-7.0295 0.6197 0.1446>, rot=<-0.0489 -0.0252 0.1892 -0.9804>}
 ring_02_r {parent=ring_01_r, loc=<-3.6174 -0.4647 0.1027>, rot=<0 0 0.2806 -0.9598>}
 ring_03_r {parent=ring_02_r, loc=<-2.4152 -0.114 -0.0199>, rot=<0 0 0.1189 -0.9929>}
 pinky_metacarpal_r {parent=hand_r, loc=<-1.7616 0.9753 -3.3814>, rot=<-0.0397 0.0814 -0.069 -0.9935>}
 pinky_01_r {parent=pinky_metacarpal_r, loc=<-7.1709 0.4831 0.1094>, rot=<-0.0654 -0.0035 0.2242 -0.9723>}
 pinky_02_r {parent=pinky_01_r, loc=<-2.8274 -0.0893 0.1043>, rot=<0.0037 0.0011 0.2854 -0.9584>}
 pinky_03_r {parent=pinky_02_r, loc=<-2.2046 -0.0307 0.0078>, rot=<-0.0002 0.0098 0.1445 -0.9895>}
 deform_wrist_dn_base_r {parent=hand_r, loc=<-0.0004 2.2345 -0.0001>, rot=<0 0 0 -1>}
 deform_wrist_dn_r {parent=deform_wrist_dn_base_r, loc=<0 0 0>, rot=<0 0 0 1>}
 deform_wrist_up_base_r {parent=hand_r, loc=<0.0003 -2.2684 0.0001>, rot=<0 0 0 -1>}
 deform_wrist_up_r {parent=deform_wrist_up_base_r, loc=<0 0 0>, rot=<0 0 0 1>}
 lowerarm_twist_02_r {parent=lowerarm_r, loc=<-9.6199 0 0>, rot=<0 0 0 -1>}
 lowerarm_twist_01_r {parent=lowerarm_r, loc=<-17.8549 0 0>, rot=<0 0 0 -1>}
 deform_elbow_in_base_r {parent=lowerarm_r, loc=<-0.2111 -5.2591 -0.8275>, rot=<0 0 0 -1>}
 deform_elbow_in_r {parent=deform_elbow_in_base_r, loc=<-0.0001 0 0.0002>, rot=<0 0 0 1>}
 deform_elbow_out_base_r {parent=lowerarm_r, loc=<0 4 -0.0001>, rot=<0 0 0 -1>}
 deform_elbow_out_r {parent=deform_elbow_out_base_r, loc=<0 0 0>, rot=<0 0 0 1>}
 upperarm_twist_01_r {parent=upperarm_r, loc=<-0.0041 0 0>, rot=<0 0 0 -1>}
 deform_upperarm_bk_base_r {parent=upperarm_twist_01_r, loc=<-1.0814 7.2192 1.9455>, rot=<0 0 0 1>}
 deform_upperarm_bk_r {parent=deform_upperarm_bk_base_r, loc=<0 0 0>, rot=<0 0 0 1>}
 deform_upperarm_fr_base_r {parent=upperarm_twist_01_r, loc=<0.3689 -7.0506 2.7652>, rot=<0 0 0 1>}
 deform_upperarm_fr_r {parent=deform_upperarm_fr_base_r, loc=<0 0 0>, rot=<0 0 0 1>}
 deform_upperarm_up_base_r {parent=upperarm_twist_01_r, loc=<3.2214 1.1391 -3.3308>, rot=<0 0 0 1>}
 deform_upperarm_up_r {parent=deform_upperarm_up_base_r, loc=<0 0 0>, rot=<0 0 0 -1>}
 upperarm_twist_02_r {parent=upperarm_r, loc=<-13.7941 0 0>, rot=<0 0 0 -1>}
 dyn_collar_l_MSTR {parent=spine_05, loc=<11.6899 4.5575 -5.6139>, rot=<0.2306 -0.6122 -0.2782 -0.7033>}
 dyn_collar_l {parent=dyn_collar_l_MSTR, loc=<0 0 0>, rot=<0 0 0 1>}
 dyn_collar_l_end {parent=dyn_collar_l, loc=<4.0298 0 0>, rot=<0 0 0 1>}
 dyn_collar_r_MSTR {parent=spine_05, loc=<11.6904 4.5576 5.6139>, rot=<-0.6122 -0.2306 0.7033 -0.2782>}
 dyn_collar_r {parent=dyn_collar_r_MSTR, loc=<0 0 0>, rot=<0 0 0 1>}
 dyn_collar_r_end {parent=dyn_collar_r, loc=<-4.0297 -0.0003 0.0004>, rot=<0 0 0 1>}
 dyn_jacket_rear_r_MSTR {parent=spine_04, loc=<-1.0158 -10.7891 10.6587>, rot=<0.847 0.0083 0.015 -0.5313>}
 dyn_jacket_rear_r_01 {parent=dyn_jacket_rear_r_MSTR, loc=<0 0 0>, rot=<0 0 0 1>}
 dyn_jacket_rear_r_02 {parent=dyn_jacket_rear_r_01, loc=<-7.6752 0 0.0001>, rot=<0 0.0157 -0.0015 -0.9999>}
 dyn_jacket_rear_r_03 {parent=dyn_jacket_rear_r_02, loc=<-7.9677 0 0>, rot=<0.0001 0.0167 -0.003 -0.9999>}
 dyn_jacket_rear_r_04 {parent=dyn_jacket_rear_r_03, loc=<-7.7031 0 0>, rot=<0 0 0 1>}
 dyn_jacket_side_r_MSTR {parent=spine_04, loc=<-2.2481 -1.9572 16.8341>, rot=<0.9961 0.0276 0.0166 -0.0823>}
 dyn_jacket_side_r_01 {parent=dyn_jacket_side_r_MSTR, loc=<0 0 0>, rot=<0 0 0 1>}
 dyn_jacket_side_r_02 {parent=dyn_jacket_side_r_01, loc=<-7.2566 0 0>, rot=<0.0001 0.0153 -0.0014 -0.9999>}
 dyn_jacket_side_r_03 {parent=dyn_jacket_side_r_02, loc=<-7.4002 0 0>, rot=<0.0002 0.0209 -0.0189 -0.9996>}
 dyn_jacket_side_r_04 {parent=dyn_jacket_side_r_03, loc=<-7.8567 0 0>, rot=<0 0 0 1>}
 dyn_jacket_front_r_MSTR {parent=spine_04, loc=<3.524 9.7599 12.0661>, rot=<0.9228 0.0362 -0.0376 0.3816>}
 dyn_jacket_front_r_01 {parent=dyn_jacket_front_r_MSTR, loc=<0 0 0>, rot=<0 0 0 1>}
 dyn_jacket_front_r_02 {parent=dyn_jacket_front_r_01, loc=<-8.6334 0 0>, rot=<0 -0.0073 0.0011 -1>}
 dyn_jacket_front_r_03 {parent=dyn_jacket_front_r_02, loc=<-7.7092 0 0>, rot=<0.0002 -0.0169 0.0066 -0.9998>}
 dyn_jacket_front_r_04 {parent=dyn_jacket_front_r_03, loc=<-7.4191 0 0>, rot=<0 0 0 -1>}
 dyn_jacket_front_r_05 {parent=dyn_jacket_front_r_04, loc=<-5.9224 0 0>, rot=<0 0 0 -1>}
 dyn_jacket_front_l_MSTR {parent=spine_04, loc=<3.5242 9.7599 -12.0661>, rot=<0.0362 -0.9228 -0.3816 -0.0376>}
 dyn_jacket_front_l_01 {parent=dyn_jacket_front_l_MSTR, loc=<0 0 0>, rot=<0 0 0 1>}
 dyn_jacket_front_l_02 {parent=dyn_jacket_front_l_01, loc=<8.6331 0 0>, rot=<0 -0.0073 0.0011 -1>}
 dyn_jacket_front_l_03 {parent=dyn_jacket_front_l_02, loc=<7.7094 0 0>, rot=<0.0002 -0.0169 0.0066 -0.9998>}
 dyn_jacket_front_l_04 {parent=dyn_jacket_front_l_03, loc=<7.4192 0 0>, rot=<0 0 0 1>}
 dyn_jacket_front_l_05 {parent=dyn_jacket_front_l_04, loc=<5.9226 0 0>, rot=<0 0 0 1>}
 dyn_zipper_base {parent=dyn_jacket_front_l_04, loc=<4.452 7.6542 -1.8811>, rot=<0.2733 -0.0148 0.0046 -0.9618>}
 dyn_zipper_tab_MSTR {parent=dyn_zipper_base, loc=<0.1404 0.0001 0.3563>, rot=<0 0 0 1>}
 dyn_zipper_tab {parent=dyn_zipper_tab_MSTR, loc=<0 0 0>, rot=<0 0 0 1>}
 dyn_jacket_rear_MSTR {parent=spine_04, loc=<-0.6481 -10.7745 0.0272>, rot=<0.0164 -0.7024 0.7113 -0.0193>}
 dyn_jacket_rear_01 {parent=dyn_jacket_rear_MSTR, loc=<0 0 0>, rot=<0 0 0 1>}
 dyn_jacket_rear_02 {parent=dyn_jacket_rear_01, loc=<8.0777 0 0>, rot=<0 0.0171 0.0018 -0.9999>}
 dyn_jacket_rear_03 {parent=dyn_jacket_rear_02, loc=<7.7024 0 0>, rot=<0 0 0 1>}
 dyn_jacket_rear_04 {parent=dyn_jacket_rear_03, loc=<8.0115 0 0>, rot=<0 0 0 1>}
 dyn_jacket_rear_l_MSTR {parent=spine_04, loc=<-1.0154 -10.7891 -10.6587>, rot=<0.0083 -0.847 0.5313 0.015>}
 dyn_jacket_rear_l_01 {parent=dyn_jacket_rear_l_MSTR, loc=<0 0 0>, rot=<0 0 0 1>}
 dyn_jacket_rear_l_02 {parent=dyn_jacket_rear_l_01, loc=<7.6757 0 0>, rot=<0 0.0157 -0.0015 -0.9999>}
 dyn_jacket_rear_l_03 {parent=dyn_jacket_rear_l_02, loc=<7.9678 0 0>, rot=<0.0001 0.0167 -0.003 -0.9999>}
 dyn_jacket_rear_l_04 {parent=dyn_jacket_rear_l_03, loc=<7.703 0 0>, rot=<0 0 0 1>}
 dyn_jacket_side_l_MSTR {parent=spine_04, loc=<-2.2479 -1.9572 -16.8341>, rot=<0.0276 -0.9961 0.0823 0.0166>}
 dyn_jacket_side_l_01 {parent=dyn_jacket_side_l_MSTR, loc=<0 0 0>, rot=<0 0 0 1>}
 dyn_jacket_side_l_02 {parent=dyn_jacket_side_l_01, loc=<7.2566 0 0>, rot=<0.0001 0.0153 -0.0014 -0.9999>}
 dyn_jacket_side_l_03 {parent=dyn_jacket_side_l_02, loc=<7.4005 0 0>, rot=<0.0002 0.0209 -0.0189 -0.9996>}
 dyn_jacket_side_l_04 {parent=dyn_jacket_side_l_03, loc=<7.8567 0 0>, rot=<0 0 0 1>}
 deform_pec_aim_base_l {parent=spine_04, loc=<7.8732 12.6388 -2.4123>, rot=<0.0102 -0.4602 0.3979 -0.7936>}
 deform_pec_aim_l {parent=deform_pec_aim_base_l, loc=<0 0 0>, rot=<0 0 0 1>}
 deform_pec_base_l {parent=deform_pec_aim_l, loc=<10.766 0 0>, rot=<0 0 0 1>}
 deform_pec_l {parent=deform_pec_base_l, loc=<0.055 0 0>, rot=<0 0 0 1>}
 deform_pec_aim_base_r {parent=spine_04, loc=<7.8727 12.6388 2.4123>, rot=<-0.4602 -0.0102 0.7936 0.3979>}
 deform_pec_aim_r {parent=deform_pec_aim_base_r, loc=<0 0 0>, rot=<0 0 0 1>}
 deform_pec_base_r {parent=deform_pec_aim_r, loc=<-10.7662 -0.0006 -0.0007>, rot=<0 0 0 1>}
 deform_pec_r {parent=deform_pec_base_r, loc=<0 0 0>, rot=<0 0 0 1>}
 deform_pin_01 {parent=spine_04, loc=<1.8149 11.629 -10.2318>, rot=<0.156 -0.1284 -0.683 -0.7019>}
 deform_pin_02 {parent=spine_04, loc=<1.4143 9.0477 -14.3697>, rot=<0.2714 -0.2536 -0.6881 -0.6233>}
 deform_pin_03 {parent=spine_04, loc=<6.1574 10.2139 -12.5097>, rot=<0.1909 -0.1506 -0.662 -0.709>}
 thigh_l {parent=pelvis, loc=<-2.5693 -0.0109 -7.7501>, rot=<0.0012 0.0572 0.0155 -0.9982>}
 calf_l {parent=thigh_l, loc=<-39.0484 0 0>, rot=<0.0006 -0.0161 -0.0391 -0.9991>}
 foot_l {parent=calf_l, loc=<-42.561 0.1077 0.4292>, rot=<0.0418 -0.0455 0.0343 -0.9975>}
 ball_l {parent=foot_l, loc=<-6.1599 13.8264 0.6806>, rot=<0.0001 -0.0001 -0.7063 -0.7079>}
 calf_twist_02_l {parent=calf_l, loc=<-18.8064 -0.0967 0.0397>, rot=<0.0237 0.0095 -0.0035 -0.9997>}
 dyn_pants_l_MSTR {parent=calf_twist_02_l, loc=<0 0 0>, rot=<0 0 0 1>}
 dyn_pants_in_l_MSTR {parent=dyn_pants_l_MSTR, loc=<-1.9463 -1.9505 5.877>, rot=<0.0516 -0.0266 0.9973 0.0439>}
 dyn_pants_in_l_01 {parent=dyn_pants_in_l_MSTR, loc=<0 0 0>, rot=<0 0 0 1>}
 dyn_pants_in_l_02 {parent=dyn_pants_in_l_01, loc=<10.147 0 0>, rot=<0 0 0 1>}
 dyn_pants_in_l_03 {parent=dyn_pants_in_l_02, loc=<8.942 0 0>, rot=<0 0 0 1>}
 dyn_pants_fr_l_MSTR {parent=dyn_pants_l_MSTR, loc=<-2.0134 6.1395 -1.6577>, rot=<-0.0672 -0.7174 -0.6934 0.0083>}
 dyn_pants_fr_l_01 {parent=dyn_pants_fr_l_MSTR, loc=<0 0 0>, rot=<0 0 0 1>}
 dyn_pants_fr_l_02 {parent=dyn_pants_fr_l_01, loc=<6.9232 0 0>, rot=<0 0 0 1>}
 dyn_pants_fr_l_03 {parent=dyn_pants_fr_l_02, loc=<6.7184 0 0>, rot=<0 0 0 1>}
 dyn_pants_out_l_MSTR {parent=dyn_pants_l_MSTR, loc=<-3.0354 -1.3024 -6.397>, rot=<-0.0439 -0.9973 -0.0266 0.0516>}
 dyn_pants_out_l_01 {parent=dyn_pants_out_l_MSTR, loc=<0 0 0>, rot=<0 0 0 1>}
 dyn_pants_out_l_02 {parent=dyn_pants_out_l_01, loc=<10.3001 0 0>, rot=<0 0 0 1>}
 dyn_pants_out_l_03 {parent=dyn_pants_out_l_02, loc=<10.2261 0 0>, rot=<0 0 0 1>}
 dyn_pants_rear_l_MSTR {parent=dyn_pants_l_MSTR, loc=<-3.2854 -7.5976 -0.9111>, rot=<0.0054 -0.724 0.6865 0.0675>}
 dyn_pants_rear_l_01 {parent=dyn_pants_rear_l_MSTR, loc=<0 0 0>, rot=<0 0 0 1>}
 dyn_pants_rear_l_02 {parent=dyn_pants_rear_l_01, loc=<10.1019 0 0>, rot=<0 0 0 1>}
 dyn_pants_rear_l_03 {parent=dyn_pants_rear_l_02, loc=<10.4075 0 0>, rot=<0 0 0 1>}
 calf_twist_01_l {parent=calf_l, loc=<-35.813 -0.0967 0.0397>, rot=<0.0237 0.0095 -0.0035 -0.9997>}
 deform_knee_in_base_l {parent=calf_l, loc=<1.298 -5.2957 -0.0496>, rot=<0 0 0 -1>}
 deform_knee_in_l {parent=deform_knee_in_base_l, loc=<0 0 0>, rot=<0 0 0 1>}
 deform_knee_out_base_l {parent=calf_l, loc=<0 4 0>, rot=<0 0 0 -1>}
 deform_knee_out_l {parent=deform_knee_out_base_l, loc=<0 0 0>, rot=<0 0 0 1>}
 thigh_twist_01_l {parent=thigh_l, loc=<-21.6303 0 0>, rot=<0.0013 0 0 -1>}
 deform_groin_base_l {parent=thigh_l, loc=<-2.1908 8.1142 -1.7442>, rot=<0 0 0 -1>}
 deform_groin_l {parent=deform_groin_base_l, loc=<0 0 0>, rot=<0 0 0 1>}
 thigh_r {parent=pelvis, loc=<-2.5693 -0.0109 7.7501>, rot=<0.0572 -0.0012 0.9982 0.0155>}
 calf_r {parent=thigh_r, loc=<39.0484 0 0>, rot=<0.0006 -0.0161 -0.0391 -0.9991>}
 foot_r {parent=calf_r, loc=<42.561 -0.1077 -0.4292>, rot=<0.0418 -0.0455 0.0343 -0.9975>}
 ball_r {parent=foot_r, loc=<6.1599 -13.8264 -0.6806>, rot=<0.0001 -0.0001 -0.7063 -0.7079>}
 calf_twist_02_r {parent=calf_r, loc=<18.8064 0.0967 -0.0397>, rot=<0.0237 0.0095 -0.0035 -0.9997>}
 dyn_pants_r_MSTR {parent=calf_twist_02_r, loc=<0 0 0>, rot=<0 0 0 1>}
 dyn_pants_rear_r_MSTR {parent=dyn_pants_r_MSTR, loc=<2.7937 7.7213 -0.4761>, rot=<0.6865 -0.0675 -0.0054 -0.724>}
 dyn_pants_rear_r_01 {parent=dyn_pants_rear_r_MSTR, loc=<0 0 0>, rot=<0 0 0 1>}
 dyn_pants_rear_r_02 {parent=dyn_pants_rear_r_01, loc=<10.4505 0 0>, rot=<0 0 0 1>}
 dyn_pants_rear_r_03 {parent=dyn_pants_rear_r_02, loc=<11.0859 0 0>, rot=<0 0 0 1>}
 dyn_pants_front_r_MSTR {parent=dyn_pants_r_MSTR, loc=<2.059 -5.5993 1.8703>, rot=<-0.724 -0.0054 0.0675 -0.6865>}
 dyn_pants_front_r_01 {parent=dyn_pants_front_r_MSTR, loc=<0 0 0>, rot=<0 0 0 1>}
 dyn_pants_front_r_02 {parent=dyn_pants_front_r_01, loc=<6.9173 0 0>, rot=<0 0 0 1>}
 dyn_pants_front_r_03 {parent=dyn_pants_front_r_02, loc=<7.6767 0 0>, rot=<0 0 0 1>}
 dyn_pants_out_r_MSTR {parent=dyn_pants_r_MSTR, loc=<3.0199 2.4387 5.9192>, rot=<-0.0266 -0.0516 0.0439 -0.9973>}
 dyn_pants_out_r_01 {parent=dyn_pants_out_r_MSTR, loc=<0 0 0>, rot=<0 0 0 1>}
 dyn_pants_out_r_02 {parent=dyn_pants_out_r_01, loc=<10.4127 0 0>, rot=<0 0 0 1>}
 dyn_pants_out_r_03 {parent=dyn_pants_out_r_02, loc=<10.0036 0 0>, rot=<0 0 0 1>}
 dyn_pants_in_r_MSTR {parent=dyn_pants_r_MSTR, loc=<1.9784 -0.1701 -5.9927>, rot=<0.9973 -0.0439 -0.0516 -0.0266>}
 dyn_pants_in_r_01 {parent=dyn_pants_in_r_MSTR, loc=<0 0 0>, rot=<0 0 0 1>}
 dyn_pants_in_r_02 {parent=dyn_pants_in_r_01, loc=<9.3067 0 0>, rot=<0 0 0 1>}
 dyn_pants_in_r_03 {parent=dyn_pants_in_r_02, loc=<8.9535 0 0>, rot=<0 0 0 1>}
 calf_twist_01_r {parent=calf_r, loc=<35.813 0.0967 -0.0397>, rot=<0.0237 0.0095 -0.0035 -0.9997>}
 deform_knee_in_base_r {parent=calf_r, loc=<-1.298 5.2957 0.0496>, rot=<0 0 0 -1>}
 deform_knee_in_r {parent=deform_knee_in_base_r, loc=<0 0 0>, rot=<0 0 0 1>}
 deform_knee_out_base_r {parent=calf_r, loc=<0 -4 0>, rot=<0 0 0 -1>}
 deform_knee_out_r {parent=deform_knee_out_base_r, loc=<0 0 0>, rot=<0 0 0 1>}
 thigh_twist_01_r {parent=thigh_r, loc=<21.6303 0 0>, rot=<0.0013 0 0 -1>}
 deform_groin_base_r {parent=thigh_r, loc=<2.1908 -8.1142 1.7442>, rot=<0 0 0 -1>}
 deform_groin_r {parent=deform_groin_base_r, loc=<0 0 0>, rot=<0 0 0 1>}
 deform_glute_base_l {parent=pelvis, loc=<-8.8064 -8.8222 -8.4287>, rot=<0.0275 0.0501 -0.4472 -0.8926>}
 deform_glute_l {parent=deform_glute_base_l, loc=<0 0 0>, rot=<0 0 0 1>}
 deform_glute_base_r {parent=pelvis, loc=<-8.8062 -8.8222 8.4287>, rot=<0.0501 -0.0275 0.8926 -0.4472>}
 deform_glute_r {parent=deform_glute_base_r, loc=<0 0 0>, rot=<0 0 0 1>}
]
variables [
 JacketAdjustParachuting {type=FTransform, default=(Rotation=(X=0,Y=-0.1736,Z=0,W=0.9848),Translation=(X=0,Y=0,Z=3),Scale3D=(X=1,Y=1,Z=1))}
 JacketAdjustSkydiving {type=FTransform, default=(Rotation=(X=0,Y=0,Z=0,W=1),Translation=(X=0,Y=0,Z=0),Scale3D=(X=1,Y=1,Z=1))}
 PantsHarmonicsSpeed {type=double, default=0}
 Parachuting {type=double, default=0}
 Skydiving {type=double, default=0}
 SprintJump {type=double, default=0}
]
