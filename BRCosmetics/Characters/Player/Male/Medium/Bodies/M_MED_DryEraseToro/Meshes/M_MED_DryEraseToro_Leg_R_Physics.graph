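# Physics asset for the right leg of the DryEraseToro male medium player cosmetic, covering the dynamic pant-leg bones.
# Front, rear, inner and outer chains each link three capsule bodies to a tiny master body through tight swing and twist limits.
# The chain masters, overall pants master and foot capsule have no constraints joining them to one another.
# The foot_r shape is named foot_l_capsule despite sitting on the right foot.
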PhysicsAsset /BRCosmetics/Characters/Player/Male/Medium/Bodies/M_MED_DryEraseToro/Meshes/M_MED_DryEraseToro_Leg_R_Physics
bodies {
 dyn_pants_front_r_01 [{shapeType=Capsule, length=10, radius=5, shapeName=dyn_pants_front_r_01_capsule, center=<3 0 3>, rotation=<rot 0 0 90>}]
 dyn_pants_front_r_02 [{shapeType=Capsule, length=10, radius=5, shapeName=dyn_pants_front_r_02_capsule, center=<3 0 3>, rotation=<rot 0 0 90>}]
 dyn_pants_front_r_03 [{shapeType=Capsule, length=10, radius=5, shapeName=dyn_pants_front_r_03_capsule, center=<3 0 3>, rotation=<rot 0 0 90>}]
 dyn_pants_front_r_MSTR [{shapeType=Capsule, length=0.505, radius=0.505, shapeName=dyn_pants_front_r_MSTR_capsule, center=<3 0 0>, rotation=<rot 0 0 90>}]
 dyn_pants_in_r_01 [{shapeType=Capsule, length=10, radius=5, shapeName=dyn_pants_in_r_01_capsule, center=<3 0 3>, rotation=<rot 0 0 90>}]
 dyn_pants_in_r_02 [{shapeType=Capsule, length=10, radius=5, shapeName=dyn_pants_in_r_02_capsule, center=<3 0 3>, rotation=<rot 0 0 90>}]
 dyn_pants_in_r_03 [{shapeType=Capsule, length=10, radius=5, shapeName=dyn_pants_in_r_03_capsule, center=<3 0 3>, rotation=<rot 0 0 90>}]
 dyn_pants_in_r_MSTR [{shapeType=Capsule, length=0.505, radius=0.505, shapeName=dyn_pants_in_r_MSTR_capsule, center=<3 0 0>, rotation=<rot 0 0 90>}]
 dyn_pants_out_r_01 [{shapeType=Capsule, length=10, radius=5, shapeName=dyn_pants_out_r_01_capsule, center=<3 0 3>, rotation=<rot 0 0 90>}]
 dyn_pants_out_r_02 [{shapeType=Capsule, length=10, radius=5, shapeName=dyn_pants_out_r_02_capsule, center=<3 0 3>, rotation=<rot 0 0 90>}]
 dyn_pants_out_r_03 [{shapeType=Capsule, length=10, radius=5, shapeName=dyn_pants_out_r_03_capsule, center=<3 0 3>, rotation=<rot 0 0 90>}]
 dyn_pants_out_r_MSTR [{shapeType=Capsule, length=0.505, radius=0.505, shapeName=dyn_pants_out_r_MSTR_capsule, center=<3 0 0>, rotation=<rot 0 0 90>}]
 dyn_pants_r_MSTR [{shapeType=Capsule, length=0.505, radius=0.505, shapeName=dyn_pants_r_MSTR_capsule, center=<3 0 0>, rotation=<rot 0 0 90>}]
 dyn_pants_rear_r_01 [{shapeType=Capsule, length=10, radius=5, shapeName=dyn_pants_rear_r_01_capsule, center=<3 0 3>, rotation=<rot 0 0 90>}]
 dyn_pants_rear_r_02 [{shapeType=Capsule, length=10, radius=5, shapeName=dyn_pants_rear_r_02_capsule, center=<3 0 3>, rotation=<rot 0 0 90>}]
 dyn_pants_rear_r_03 [{shapeType=Capsule, length=10, radius=5, shapeName=dyn_pants_rear_r_03_capsule, center=<3 0 3>, rotation=<rot 0 0 90>}]
 dyn_pants_rear_r_MSTR [{shapeType=Capsule, length=0.505, radius=0.505, shapeName=dyn_pants_rear_r_MSTR_capsule, center=<3 0 0>, rotation=<rot 0 0 90>}]
 foot_r [{shapeType=Capsule, length=10.5581, radius=5, shapeName=foot_l_capsule, center=<0 0.6397 0>, rotation=<rot 90 0 0>}]
}
constraints [
 {bone1Name=dyn_pants_front_r_01, bone2Name=dyn_pants_front_r_MSTR, swing1LimitDegrees=6, swing2LimitDegrees=6, twistLimitDegrees=4}
 {bone1Name=dyn_pants_front_r_02, bone2Name=dyn_pants_front_r_01, swing1LimitDegrees=6, swing2LimitDegrees=6, twistLimitDegrees=4}
 {bone1Name=dyn_pants_front_r_03, bone2Name=dyn_pants_front_r_02, swing1LimitDegrees=6, swing2LimitDegrees=6, twistLimitDegrees=4}
 {bone1Name=dyn_pants_in_r_01, bone2Name=dyn_pants_in_r_MSTR, swing1LimitDegrees=6, swing2LimitDegrees=6, twistLimitDegrees=4}
 {bone1Name=dyn_pants_in_r_02, bone2Name=dyn_pants_in_r_01, swing1LimitDegrees=6, swing2LimitDegrees=6, twistLimitDegrees=4}
 {bone1Name=dyn_pants_in_r_03, bone2Name=dyn_pants_in_r_02, swing1LimitDegrees=6, swing2LimitDegrees=6, twistLimitDegrees=4}
 {bone1Name=dyn_pants_out_r_01, bone2Name=dyn_pants_out_r_MSTR, swing1LimitDegrees=6, swing2LimitDegrees=6, twistLimitDegrees=4}
 {bone1Name=dyn_pants_out_r_02, bone2Name=dyn_pants_out_r_01, swing1LimitDegrees=6, swing2LimitDegrees=6, twistLimitDegrees=4}
 {bone1Name=dyn_pants_out_r_03, bone2Name=dyn_pants_out_r_02, swing1LimitDegrees=6, swing2LimitDegrees=6, twistLimitDegrees=4}
 {bone1Name=dyn_pants_rear_r_01, bone2Name=dyn_pants_rear_r_MSTR, swing1LimitDegrees=6, swing2LimitDegrees=6, twistLimitDegrees=4}
 {bone1Name=dyn_pants_rear_r_02, bone2Name=dyn_pants_rear_r_01, swing1LimitDegrees=6, swing2LimitDegrees=6, twistLimitDegrees=4}
 {bone1Name=dyn_pants_rear_r_03, bone2Name=dyn_pants_rear_r_02, swing1LimitDegrees=6, swing2LimitDegrees=6, twistLimitDegrees=4}
]
