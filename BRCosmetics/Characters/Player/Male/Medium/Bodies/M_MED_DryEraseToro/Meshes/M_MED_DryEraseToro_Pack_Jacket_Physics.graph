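# Physics asset for the DryEraseToro male medium character's jacket, simulating the collar and jacket panels.
# Capsule chains hang from MSTR bones for both collars and the front, side and rear panels, with pelvis, spine and thigh capsules as body shapes.
# Collar, front and side joints carry angular limits, while rear joints are locked.
# Free-rotation constraints link front, side and rear hem tips on each side, leaving the center-rear chain unlinked.
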PhysicsAsset /BRCosmetics/Characters/Player/Male/Medium/Bodies/M_MED_DryEraseToro/Meshes/M_MED_DryEraseToro_Pack_Jacket_Physics
bodies {
 dyn_collar_l [{shapeType=Capsule, length=4.2324, radius=2.8726, shapeName=dyn_collar_l_capsule, center=<2 0 2>, rotation=<rot 0.0001 -0.0011 79.7028>}]
 dyn_collar_l_MSTR [{shapeType=Capsule, length=0.505, radius=0.505, shapeName=dyn_collar_l_MSTR_capsule, rotation=<rot 0 -90 0>}]
 dyn_collar_l_end [{shapeType=Capsule, length=3.0336, radius=1.9097, shapeName=dyn_collar_l_end_capsule, center=<2 0 0>, rotation=<rot 0.0001 -0.0011 79.7028>}]
 dyn_collar_r [{shapeType=Capsule, length=4.2234, radius=2.8471, shapeName=dyn_collar_r_capsule, center=<-2 0 -2>, rotation=<rot 0 0 90>}]
 dyn_collar_r_MSTR [{shapeType=Capsule, length=0.505, radius=0.505, shapeName=dyn_collar_r_MSTR_capsule, rotation=<rot 0 -90 0>}]
 dyn_collar_r_end [{shapeType=Capsule, length=3.0453, radius=1.9062, shapeName=dyn_collar_r_end_capsule, center=<-2 0 0>, rotation=<rot 0 0 90>}]
 dyn_jacket_front_l_01 [{
   shapeType=Capsule
   length=19.2165
   radius=8
   shapeName=dyn_jacket_front_l_01_capsule
   center=<0 0.1921 6.8619>
   rotation=<rot 0.0001 0.0001 102.5628>
  }]
 dyn_jacket_front_l_02 [{
   shapeType=Capsule
   length=19.2165
   radius=8
   shapeName=dyn_jacket_front_l_02_capsule
   center=<0 0.2266 6.4167>
   rotation=<rot 0.0001 0.0001 102.5628>
  }]
 dyn_jacket_front_l_03 [{
   shapeType=Capsule
   length=19.2165
   radius=8
   shapeName=dyn_jacket_front_l_03_capsule
   center=<0 0.2674 6.7999>
   rotation=<rot 0.0001 0.0001 102.5628>
  }]
 dyn_jacket_front_l_04 [{
   shapeType=Capsule
   length=19.2165
   radius=8
   shapeName=dyn_jacket_front_l_04_capsule
   center=<0 0.2288 7.5267>
   rotation=<rot 0.0001 0.0001 102.5628>
  }]
 dyn_jacket_front_l_05 [{
   shapeType=Capsule
   length=19.2165
   radius=8
   shapeName=dyn_jacket_front_l_05_capsule
   center=<0 -0.2756 7.2463>
   rotation=<rot 0.0003 0.0003 89.5211>
  }]
 dyn_jacket_front_l_MSTR [{shapeType=Capsule, length=0.505, radius=0.1, shapeName=dyn_jacket_front_l_MSTR_capsule, rotation=<rot 0 -90 0>}]
 dyn_jacket_front_r_01 [{
   shapeType=Capsule
   length=19.2165
   radius=8
   shapeName=dyn_jacket_front_r_01_capsule
   center=<0.0001 -0.2546 -6.7737>
   rotation=<rot 0.0081 0.0081 105.3972>
  }]
 dyn_jacket_front_r_02 [{
   shapeType=Capsule
   length=19.2165
   radius=8
   shapeName=dyn_jacket_front_r_02_capsule
   center=<0.0003 -0.3844 -6.7959>
   rotation=<rot 0.0081 0.0081 105.3973>
  }]
 dyn_jacket_front_r_03 [{
   shapeType=Capsule
   length=19.2165
   radius=8
   shapeName=dyn_jacket_front_r_03_capsule
   center=<0.0002 -0.4025 -7.1369>
   rotation=<rot 0.0081 0.0082 105.3974>
  }]
 dyn_jacket_front_r_04 [{
   shapeType=Capsule
   length=19.2165
   radius=8
   shapeName=dyn_jacket_front_r_04_capsule
   center=<0.0002 -0.4973 -7.3547>
   rotation=<rot 0.0081 0.0082 96.5548>
  }]
 dyn_jacket_front_r_05 [
  {
   shapeType=Capsule
   length=19.2165
   radius=8
   shapeName=dyn_jacket_front_r_05_capsule
   center=<0.0006 -0.0976 -6.8543>
   rotation=<rot 0.0394 0.0399 95.0866>
  }
  {
   shapeType=Capsule
   length=19.2165
   radius=8
   shapeName=dyn_jacket_front_r_05_capsule2
   center=<0.1094 6.4953 -4.9664>
   rotation=<rot 0.0394 0.0399 73.9819>
  }
 ]
 dyn_jacket_front_r_MSTR [{shapeType=Capsule, length=0.505, radius=0.1, shapeName=dyn_jacket_front_r_MSTR_capsule, rotation=<rot 0 -90 0>}]
 dyn_jacket_rear_01 [{shapeType=Capsule, length=19.2165, radius=8, shapeName=dyn_jacket_rear_01_capsule, center=<0.1832 0 7.6109>, rotation=<rot 0 0 90>}]
 dyn_jacket_rear_02 [{shapeType=Capsule, length=19.2165, radius=8, shapeName=dyn_jacket_rear_02_capsule, center=<0.1832 0 7.8841>, rotation=<rot 0 0 90>}]
 dyn_jacket_rear_03 [{shapeType=Capsule, length=18.5873, radius=8.6292, shapeName=dyn_jacket_rear_03_capsule, center=<0.1832 0 8.3312>, rotation=<rot 0 0 90>}]
 dyn_jacket_rear_04 [{shapeType=Capsule, length=18.9298, radius=8.2867, shapeName=dyn_jacket_rear_04_capsule, center=<0.1832 0 7.1392>, rotation=<rot 0 0 90>}]
 dyn_jacket_rear_MSTR [{shapeType=Capsule, length=0.505, radius=0.1, shapeName=dyn_jacket_rear_MSTR_capsule, rotation=<rot 0 -90 0>}]
 dyn_jacket_rear_l_01 [{shapeType=Capsule, length=19.2165, radius=8, shapeName=dyn_jacket_rear_l_01_capsule, center=<0 0 6.6778>, rotation=<rot 0 0 90>}]
 dyn_jacket_rear_l_02 [{shapeType=Capsule, length=19.2165, radius=8, shapeName=dyn_jacket_rear_l_02_capsule, center=<0 0 6.7117>, rotation=<rot 0 0 90>}]
 dyn_jacket_rear_l_03 [{shapeType=Capsule, length=19.2165, radius=8, shapeName=dyn_jacket_rear_l_03_capsule, center=<0 0 7.039>, rotation=<rot 0 0 90>}]
 dyn_jacket_rear_l_04 [
  {shapeType=Capsule, length=19.2165, radius=8, shapeName=dyn_jacket_rear_l_04_capsule, center=<-0.0649 0 6.3338>, rotation=<rot 0 0 90>}
  {
   shapeType=Capsule
   length=19.2165
   radius=8
   shapeName=dyn_jacket_rear_l_04_capsule2
   center=<0.0135 -5.8426 5.5699>
   rotation=<rot -0.0001 0 76.9742>
  }
 ]
 dyn_jacket_rear_l_MSTR [{shapeType=Capsule, length=0.505, radius=0.1, shapeName=dyn_jacket_rear_l_MSTR_capsule, rotation=<rot 0 -90 0>}]
 dyn_jacket_rear_r_01 [{shapeType=Capsule, length=19.2165, radius=8, shapeName=dyn_jacket_rear_r_01_capsule, center=<0 0 -6.5>, rotation=<rot -0.0001 0 95.3737>}]
 dyn_jacket_rear_r_02 [{shapeType=Capsule, length=19.2165, radius=8, shapeName=dyn_jacket_rear_r_02_capsule, center=<0 0 -6.6>, rotation=<rot -0.0001 0 95.3737>}]
 dyn_jacket_rear_r_03 [{
   shapeType=Capsule
   length=18.6364
   radius=8.5801
   shapeName=dyn_jacket_rear_r_03_capsule
   center=<0 -0.0735 -7.7812>
   rotation=<rot -0.0001 0 95.3737>
  }]
 dyn_jacket_rear_r_04 [{
   shapeType=Capsule
   length=18.3881
   radius=8.8284
   shapeName=dyn_jacket_rear_r_04_capsule
   center=<0 -0.0272 -7.4888>
   rotation=<rot -0.0001 0 95.3737>
  }]
 dyn_jacket_rear_r_MSTR [{shapeType=Capsule, length=0.505, radius=0.1, shapeName=dyn_jacket_rear_r_MSTR_capsule, rotation=<rot 0 -90 0>}]
 dyn_jacket_side_l_01 [{shapeType=Capsule, length=19.2165, radius=8, shapeName=dyn_jacket_side_l_01_capsule, center=<0 0.0197 6.1905>, rotation=<rot 0 0 87.3624>}]
 dyn_jacket_side_l_02 [{shapeType=Capsule, length=19.2165, radius=8, shapeName=dyn_jacket_side_l_02_capsule, center=<0 0.0197 5.7905>, rotation=<rot 0 0 87.3624>}]
 dyn_jacket_side_l_03 [{shapeType=Capsule, length=19.2165, radius=8, shapeName=dyn_jacket_side_l_03_capsule, center=<0 0.0197 5.8905>, rotation=<rot 0 0 87.3624>}]
 dyn_jacket_side_l_04 [{
   shapeType=Capsule
   length=18.3727
   radius=8.8438
   shapeName=dyn_jacket_side_l_04_capsule
   center=<0 0.0197 7.0905>
   rotation=<rot 0 0 87.3624>
  }]
 dyn_jacket_side_l_MSTR [{shapeType=Capsule, length=0.505, radius=0.1, shapeName=dyn_jacket_side_l_MSTR_capsule, rotation=<rot 0 -90 0>}]
 dyn_jacket_side_r_01 [{
   shapeType=Capsule
   length=19.2165
   radius=8
   shapeName=dyn_jacket_side_r_01_capsule
   center=<0 0 -7.1>
   rotation=<rot -0.0001 0.0003 92.2477>
  }]
 dyn_jacket_side_r_02 [{
   shapeType=Capsule
   length=19.2165
   radius=8
   shapeName=dyn_jacket_side_r_02_capsule
   center=<0 0 -6.4>
   rotation=<rot -0.0001 0.0003 92.2477>
  }]
 dyn_jacket_side_r_03 [{
   shapeType=Capsule
   length=18.7571
   radius=8.4594
   shapeName=dyn_jacket_side_r_03_capsule
   center=<0 0 -6.7>
   rotation=<rot -0.0001 0.0003 92.2476>
  }]
 dyn_jacket_side_r_04 [{
   shapeType=Capsule
   length=18.0134
   radius=9.2031
   shapeName=dyn_jacket_side_r_04_capsule
   center=<0 0 -7>
   rotation=<rot -0.0001 0.0003 92.2476>
  }]
 dyn_jacket_side_r_MSTR [{shapeType=Capsule, length=0.505, radius=0.1, shapeName=dyn_jacket_side_r_MSTR_capsule, rotation=<rot 0 -90 0>}]
 pelvis [
  {shapeType=Capsule, length=9.5908, radius=11.7068, shapeName=pelvis_capsule, center=<1.3336 -0.4942 0>}
  {shapeType=Capsule, length=9.3506, radius=11.9471, shapeName=pelvis_capsule2, center=<-2.8789 -0.4942 0>}
 ]
 spine_01 [{shapeType=Capsule, length=9.7285, radius=11.7404, shapeName=spine_04_capsule, center=<-2.0158 0 0>}]
 spine_02 [{shapeType=Capsule, length=9.5124, radius=11.3482, shapeName=spine_04_capsule, center=<-2.0807 -0.4454 0>}]
 spine_03 [{shapeType=Capsule, length=9.3042, radius=10.8925, shapeName=spine_04_capsule, center=<0 -0.4349 0>}]
 spine_04 [{shapeType=Capsule, length=12.1901, radius=10.3864, shapeName=spine_04_capsule}]
 spine_05 [{shapeType=Capsule, length=11.994, radius=11.8419, shapeName=spine_05_capsule, rotation=<rot 0.1595 -90 -0.0046>}]
 thigh_l [{shapeType=Capsule, length=42.176, radius=9.8461, shapeName=thigh_l_capsule, center=<-22.0098 0 0.7838>, rotation=<rot 90 0 0>}]
 thigh_r [{shapeType=Capsule, length=41.0098, radius=9.7132, shapeName=thigh_r_capsule, center=<21.1054 0 -1.1025>, rotation=<rot 90 0 0>}]
}
constraints [
 {bone1Name=dyn_jacket_front_l_01, bone2Name=dyn_jacket_front_l_MSTR, swing1LimitDegrees=8, swing2LimitDegrees=10, twistLimitDegrees=4}
 {bone1Name=dyn_jacket_front_l_02, bone2Name=dyn_jacket_front_l_01, swing1LimitDegrees=8, swing2LimitDegrees=10, twistLimitDegrees=4}
 {bone1Name=dyn_jacket_front_l_03, bone2Name=dyn_jacket_front_l_02, swing1LimitDegrees=8, swing2LimitDegrees=10, twistLimitDegrees=4}
 {bone1Name=dyn_jacket_front_l_04, bone2Name=dyn_jacket_front_l_03, swing1LimitDegrees=8, swing2LimitDegrees=10, twistLimitDegrees=4}
 {bone1Name=dyn_jacket_front_l_05, bone2Name=dyn_jacket_front_l_04, swing1LimitDegrees=8, swing2LimitDegrees=10, twistLimitDegrees=4}
 {bone1Name=dyn_jacket_front_r_01, bone2Name=dyn_jacket_front_r_MSTR, swing1LimitDegrees=8, swing2LimitDegrees=10, twistLimitDegrees=4}
 {bone1Name=dyn_jacket_front_r_02, bone2Name=dyn_jacket_front_r_01, swing1LimitDegrees=8, swing2LimitDegrees=10, twistLimitDegrees=4}
 {bone1Name=dyn_jacket_front_r_03, bone2Name=dyn_jacket_front_r_02, swing1LimitDegrees=8, swing2LimitDegrees=10, twistLimitDegrees=4}
 {bone1Name=dyn_jacket_front_r_04, bone2Name=dyn_jacket_front_r_03, swing1LimitDegrees=8, swing2LimitDegrees=10, twistLimitDegrees=4}
 {bone1Name=dyn_jacket_front_r_05, bone2Name=dyn_jacket_front_r_04, swing1LimitDegrees=8, swing2LimitDegrees=10, twistLimitDegrees=4}
 {
  bone1Name=dyn_jacket_rear_01
  bone2Name=dyn_jacket_rear_MSTR
  swing1LimitDegrees=8
  swing1Motion=Locked
  swing2LimitDegrees=20
  swing2Motion=Locked
  twistLimitDegrees=4
  twistMotion=Locked
 }
 {
  bone1Name=dyn_jacket_rear_02
  bone2Name=dyn_jacket_rear_01
  swing1LimitDegrees=8
  swing1Motion=Locked
  swing2LimitDegrees=20
  swing2Motion=Locked
  twistLimitDegrees=4
  twistMotion=Locked
 }
 {
  bone1Name=dyn_jacket_rear_03
  bone2Name=dyn_jacket_rear_02
  swing1LimitDegrees=1
  swing1Motion=Locked
  swing2LimitDegrees=1
  swing2Motion=Locked
  twistLimitDegrees=4
  twistMotion=Locked
 }
 {
  bone1Name=dyn_jacket_rear_04
  bone2Name=dyn_jacket_rear_03
  swing1LimitDegrees=1
  swing1Motion=Locked
  swing2LimitDegrees=1
  swing2Motion=Locked
  twistLimitDegrees=4
  twistMotion=Locked
 }
 {
  bone1Name=dyn_jacket_rear_l_01
  bone2Name=dyn_jacket_rear_l_MSTR
  swing1LimitDegrees=8
  swing1Motion=Locked
  swing2LimitDegrees=20
  swing2Motion=Locked
  twistLimitDegrees=4
  twistMotion=Locked
 }
 {
  bone1Name=dyn_jacket_rear_l_02
  bone2Name=dyn_jacket_rear_l_01
  swing1LimitDegrees=8
  swing1Motion=Locked
  swing2LimitDegrees=20
  swing2Motion=Locked
  twistLimitDegrees=4
  twistMotion=Locked
 }
 {
  bone1Name=dyn_jacket_rear_l_03
  bone2Name=dyn_jacket_rear_l_02
  swing1LimitDegrees=1
  swing1Motion=Locked
  swing2LimitDegrees=1
  swing2Motion=Locked
  twistLimitDegrees=4
  twistMotion=Locked
 }
 {
  bone1Name=dyn_jacket_rear_l_04
  bone2Name=dyn_jacket_rear_l_03
  swing1LimitDegrees=1
  swing1Motion=Locked
  swing2LimitDegrees=1
  swing2Motion=Locked
  twistLimitDegrees=4
  twistMotion=Locked
 }
 {
  bone1Name=dyn_jacket_rear_r_01
  bone2Name=dyn_jacket_rear_r_MSTR
  swing1LimitDegrees=8
  swing1Motion=Locked
  swing2LimitDegrees=20
  swing2Motion=Locked
  twistLimitDegrees=4
  twistMotion=Locked
 }
 {
  bone1Name=dyn_jacket_rear_r_02
  bone2Name=dyn_jacket_rear_r_01
  swing1LimitDegrees=8
  swing1Motion=Locked
  swing2LimitDegrees=20
  swing2Motion=Locked
  twistLimitDegrees=4
  twistMotion=Locked
 }
 {
  bone1Name=dyn_jacket_rear_r_03
  bone2Name=dyn_jacket_rear_r_02
  swing1LimitDegrees=1
  swing1Motion=Locked
  swing2LimitDegrees=1
  swing2Motion=Locked
  twistLimitDegrees=4
  twistMotion=Locked
 }
 {
  bone1Name=dyn_jacket_rear_r_04
  bone2Name=dyn_jacket_rear_r_03
  swing1LimitDegrees=1
  swing1Motion=Locked
  swing2LimitDegrees=1
  swing2Motion=Locked
  twistLimitDegrees=4
  twistMotion=Locked
 }
 {bone1Name=dyn_jacket_side_l_01, bone2Name=dyn_jacket_side_l_MSTR, swing1LimitDegrees=8, swing2LimitDegrees=10, twistLimitDegrees=4}
 {bone1Name=dyn_jacket_side_l_02, bone2Name=dyn_jacket_side_l_01, swing1LimitDegrees=8, swing2LimitDegrees=10, twistLimitDegrees=4}
 {bone1Name=dyn_jacket_side_l_03, bone2Name=dyn_jacket_side_l_02, swing1LimitDegrees=8, swing2LimitDegrees=10, twistLimitDegrees=4}
 {bone1Name=dyn_jacket_side_l_04, bone2Name=dyn_jacket_side_l_03, swing1LimitDegrees=8, swing2LimitDegrees=10, twistLimitDegrees=4}
 {bone1Name=dyn_jacket_side_r_01, bone2Name=dyn_jacket_side_r_MSTR, swing1LimitDegrees=8, swing2LimitDegrees=10, twistLimitDegrees=4}
 {bone1Name=dyn_jacket_side_r_02, bone2Name=dyn_jacket_side_r_01, swing1LimitDegrees=8, swing2LimitDegrees=10, twistLimitDegrees=4}
 {bone1Name=dyn_jacket_side_r_03, bone2Name=dyn_jacket_side_r_02, swing1LimitDegrees=8, swing2LimitDegrees=10, twistLimitDegrees=4}
 {bone1Name=dyn_jacket_side_r_04, bone2Name=dyn_jacket_side_r_03, swing1LimitDegrees=8, swing2LimitDegrees=10, twistLimitDegrees=4}
 {bone1Name=dyn_collar_l, bone2Name=dyn_collar_l_MSTR, swing1LimitDegrees=10, swing2LimitDegrees=30, twistLimitDegrees=2}
 {bone1Name=dyn_collar_l_end, bone2Name=dyn_collar_l, swing1LimitDegrees=10, swing2LimitDegrees=30, twistLimitDegrees=2}
 {bone1Name=dyn_collar_r, bone2Name=dyn_collar_r_MSTR, swing1LimitDegrees=10, swing2LimitDegrees=30, twistLimitDegrees=2}
 {bone1Name=dyn_collar_r_end, bone2Name=dyn_collar_r, swing1LimitDegrees=10, swing2LimitDegrees=30, twistLimitDegrees=2}
 {
  bone1Name=dyn_jacket_front_l_05
  bone2Name=dyn_jacket_side_l_04
  swing1LimitDegrees=45
  swing1Motion=Free
  swing2LimitDegrees=45
  swing2Motion=Free
  twistLimitDegrees=45
  twistMotion=Free
 }
 {
  bone1Name=dyn_jacket_front_r_05
  bone2Name=dyn_jacket_side_r_04
  swing1LimitDegrees=45
  swing1Motion=Free
  swing2LimitDegrees=45
  swing2Motion=Free
  twistLimitDegrees=45
  twistMotion=Free
 }
 {
  bone1Name=dyn_jacket_side_r_04
  bone2Name=dyn_jacket_rear_r_04
  swing1LimitDegrees=45
  swing1Motion=Free
  swing2LimitDegrees=45
  swing2Motion=Free
  twistLimitDegrees=45
  twistMotion=Free
 }
 {
  bone1Name=dyn_jacket_side_l_04
  bone2Name=dyn_jacket_rear_l_04
  swing1LimitDegrees=45
  swing1Motion=Free
  swing2LimitDegrees=45
  swing2Motion=Free
  twistLimitDegrees=45
  twistMotion=Free
 }
]
