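# Animation Blueprint for the DryEraseToro male body cosmetic, adding secondary motion to jacket, pants and zipper.
# The copied parent pose feeds a Control Rig driven by sprint-jump, skydive, parachute and speed-based values, then is cached.
# Backbling detection selects between pack and no-pack jacket RigidBody sims, followed by calf-based leg sims.
# Weighted constraints place pins and zipper, AnimDynamics swings the zipper tab, and additive bone offsets adjust the jacket.
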AnimBlueprint /BRCosmetics/Characters/Player/Male/Medium/Bodies/M_MED_DryEraseToro/Meshes/M_MED_DryEraseToro_AnimBP
target_skeleton=&/M_MED_DryEraseToro_Skeleton
graphs [
 AnimGraph {nodes [
   n1 = RigidBody!(
    ComponentPose=Local To Component(LocalPose=Use cached pose 'CR Cache'())
    Alpha
    OverrideWorldGravity:Vector="0,0,-980"
    ExternalForce
    SimSpaceSettings:Sim Space Settings Structure=(WorldAlpha=1,VelocityScaleZ=1,DampingAlpha=0,MaxLinearVelocity=800,MaxAngularVelocity=12,MaxLinearAcceleration=4000,MaxAngularAcceleration=60,ExternalLinearDragV=(X=0,Y=0,Z=0),ExternalLinearVelocity=(X=0,Y=0,Z=0),ExternalAngularVelocity=(X=0,Y=0,Z=0))
   ) -> (Pose) {
    bOverrideWorldGravity=true
    overridePhysicsAsset=&/M_MED_DryEraseToro_Pack_Jacket_Physics
    simulationSpace=BaseBoneSpace
    alphaBoolBlend {blendOption=Linear}
    baseBoneRef {boneName=spine_04}
    overrideWorldGravity {z=-980}
    simSpaceSettings {dampingAlpha=0, maxAngularAcceleration=60, maxAngularVelocity=12, maxLinearAcceleration=4000, maxLinearVelocity=800, worldAlpha=1}
   }
   n2 = Component To Local(ComponentPose=@n1)
   n3 = RigidBody(
    ComponentPose=Local To Component(LocalPose=Use cached pose 'CR Cache'())
    OverrideWorldGravity="0,0,-980"
    SimSpaceSettings=(WorldAlpha=1,VelocityScaleZ=1,DampingAlpha=0,MaxLinearVelocity=800,MaxAngularVelocity=12,MaxLinearAcceleration=4000,MaxAngularAcceleration=60,ExternalLinearDragV=(X=0,Y=0,Z=0),ExternalLinearVelocity=(X=0,Y=0,Z=0),ExternalAngularVelocity=(X=0,Y=0,Z=0))
   ) {
    bOverrideWorldGravity=true
    overridePhysicsAsset=&/M_MED_DryEraseToro_Jacket_Physics
    simulationSpace=BaseBoneSpace
    alphaBoolBlend {blendOption=Linear}
    baseBoneRef {boneName=spine_04}
    overrideWorldGravity {z=-980}
    simSpaceSettings {dampingAlpha=0, maxAngularAcceleration=60, maxAngularVelocity=12, maxLinearAcceleration=4000, maxLinearVelocity=800, worldAlpha=1}
   }
   n4 = Component To Local(ComponentPose=@n3)
   n5 = Get isWearingPack() -> (isWearingPack:bool)
   n6 = RigidBody(
    ComponentPose=Local To Component(LocalPose=Blend Poses by bool(BlendPose_0=@n2, BlendPose_1=@n4, bActiveValue=@n5) {blendType=Linear})
    OverrideWorldGravity="0,0,-980"
    SimSpaceSettings=(WorldAlpha=1,VelocityScaleZ=1,DampingAlpha=0,MaxLinearVelocity=800,MaxAngularVelocity=12,MaxLinearAcceleration=4000,MaxAngularAcceleration=60,ExternalLinearDragV=(X=0,Y=0,Z=0),ExternalLinearVelocity=(X=0,Y=0,Z=0),ExternalAngularVelocity=(X=0,Y=0,Z=0))
   ) {
    bOverrideWorldGravity=true
    overridePhysicsAsset=&/M_MED_DryEraseToro_Leg_L_Physics
    simulationSpace=BaseBoneSpace
    alphaBoolBlend {blendOption=Linear}
    baseBoneRef {boneName=calf_l}
    overrideWorldGravity {z=-980}
    simSpaceSettings {dampingAlpha=0, maxAngularAcceleration=60, maxAngularVelocity=12, maxLinearAcceleration=4000, maxLinearVelocity=800, worldAlpha=1}
   }
   n7 = RigidBody(
    ComponentPose=@n6
    OverrideWorldGravity="0,0,-980"
    SimSpaceSettings=(WorldAlpha=1,VelocityScaleZ=1,DampingAlpha=0,MaxLinearVelocity=800,MaxAngularVelocity=12,MaxLinearAcceleration=4000,MaxAngularAcceleration=60,ExternalLinearDragV=(X=0,Y=0,Z=0),ExternalLinearVelocity=(X=0,Y=0,Z=0),ExternalAngularVelocity=(X=0,Y=0,Z=0))
   ) {
    bOverrideWorldGravity=true
    overridePhysicsAsset=&/M_MED_DryEraseToro_Leg_R_Physics
    simulationSpace=BaseBoneSpace
    alphaBoolBlend {blendOption=Linear}
    baseBoneRef {boneName=calf_r}
    overrideWorldGravity {z=-980}
    simSpaceSettings {dampingAlpha=0, maxAngularAcceleration=60, maxAngularVelocity=12, maxLinearAcceleration=4000, maxLinearVelocity=800, worldAlpha=1}
   }
   n8 = "Constraint - deform_pin_01 (PPPPPPPPPPPPPPP)"!(
    ComponentPose=@n7
    Alpha=1
    ConstraintWeights_0=0.704
    ConstraintWeights_1:float=0.0983
    ConstraintWeights_2:float=0.0586
    ConstraintWeights_3:float=0.0466
    ConstraintWeights_4:float=0.0288
    ConstraintWeights_5:float=0.0281
    ConstraintWeights_6:float=0.0151
    ConstraintWeights_7:float=0.009
    ConstraintWeights_8:float=0.0052
    ConstraintWeights_9:float=0.0032
    ConstraintWeights_10:float=0.0016
    ConstraintWeights_11:float=0.0007
    ConstraintWeights_12:float=0.0005
    ConstraintWeights_13:float=0.0001
    ConstraintWeights_14:float=0.0001
   ) -> (Pose) {
    alphaBoolBlend {blendOption=Linear}
    boneToModify {boneName=deform_pin_01}
    constraintSetup [
     {targetBone {boneName=dyn_jacket_front_l_01}}
     {targetBone {boneName=deform_pec_l}}
     {targetBone {boneName=spine_03}}
     {targetBone {boneName=spine_04}}
     {targetBone {boneName=spine_02}}
     {targetBone {boneName=clavicle_l}}
     {targetBone {boneName=dyn_jacket_front_l_02}}
     {targetBone {boneName=spine_01}}
     {targetBone {boneName=spine_05}}
     {targetBone {boneName=upperarm_twist_01_l}}
     {targetBone {boneName=dyn_jacket_side_l_01}}
     {targetBone {boneName=deform_upperarm_fr_l}}
     {targetBone {boneName=dyn_jacket_front_l_03}}
     {targetBone {boneName=upperarm_l}}
     {targetBone {boneName=dyn_jacket_side_l_02}}
    ]
    constraintWeights [1,1,1,1,1,1,1,1,1,1,1,1,1,1,1]
   }
   n9 = "Constraint - deform_pin_03 (PPPPPPPPPPP)"!(
    ComponentPose=@n8
    Alpha=1
    ConstraintWeights_0=0.267
    ConstraintWeights_1:float=0.1353
    ConstraintWeights_2:float=0.135
    ConstraintWeights_3:float=0.1249
    ConstraintWeights_4:float=0.1212
    ConstraintWeights_5:float=0.0775
    ConstraintWeights_6:float=0.0554
    ConstraintWeights_7:float=0.0473
    ConstraintWeights_8:float=0.0356
    ConstraintWeights_9:float=0.0006
    ConstraintWeights_10:float=0.0003
   ) -> (Pose) {
    alphaBoolBlend {blendOption=Linear}
    boneToModify {boneName=deform_pin_03}
    constraintSetup [
     {targetBone {boneName=deform_pec_l}}
     {targetBone {boneName=spine_04}}
     {targetBone {boneName=dyn_jacket_front_l_01}}
     {targetBone {boneName=spine_03}}
     {targetBone {boneName=clavicle_l}}
     {targetBone {boneName=deform_upperarm_fr_l}}
     {targetBone {boneName=spine_02}}
     {targetBone {boneName=upperarm_twist_01_l}}
     {targetBone {boneName=spine_05}}
     {targetBone {boneName=dyn_jacket_front_l_02}}
     {targetBone {boneName=dyn_jacket_side_l_01}}
    ]
    constraintWeights [1,1,1,1,1,1,1,1,1,1,1]
   }
   n10 = "Constraint - deform_pin_02 (PPPPPPPPPPPPPP)"!(
    ComponentPose=@n9
    Alpha=1
    ConstraintWeights_0=0.6868
    ConstraintWeights_1:float=0.0547
    ConstraintWeights_2:float=0.0513
    ConstraintWeights_3:float=0.0458
    ConstraintWeights_4:float=0.0364
    ConstraintWeights_5:float=0.0348
    ConstraintWeights_6:float=0.026
    ConstraintWeights_7:float=0.0231
    ConstraintWeights_8:float=0.0189
    ConstraintWeights_9:float=0.007
    ConstraintWeights_10:float=0.0066
    ConstraintWeights_11:float=0.0043
    ConstraintWeights_12:float=0.0023
    ConstraintWeights_13:float=0.0021
   ) -> (Pose) {
    alphaBoolBlend {blendOption=Linear}
    boneToModify {boneName=deform_pin_02}
    constraintSetup [
     {targetBone {boneName=dyn_jacket_front_l_01}}
     {targetBone {boneName=deform_pec_l}}
     {targetBone {boneName=dyn_jacket_front_l_02}}
     {targetBone {boneName=spine_03}}
     {targetBone {boneName=dyn_jacket_side_l_01}}
     {targetBone {boneName=spine_04}}
     {targetBone {boneName=clavicle_l}}
     {targetBone {boneName=spine_02}}
     {targetBone {boneName=deform_upperarm_fr_l}}
     {targetBone {boneName=spine_01}}
     {targetBone {boneName=upperarm_twist_01_l}}
     {targetBone {boneName=dyn_jacket_front_l_03}}
     {targetBone {boneName=dyn_jacket_side_l_02}}
     {targetBone {boneName=upperarm_l}}
    ]
    constraintWeights [1,1,1,1,1,1,1,1,1,1,1,1,1,1]
   }
   n11 = "Constraint - dyn_zipper_base (PPPP)"!(
    ComponentPose=@n10
    Alpha=1
    ConstraintWeights_0=0.9923
    ConstraintWeights_1:float=0.0047
    ConstraintWeights_2:float=0.0025
    ConstraintWeights_3:float=0.0005
   ) -> (Pose) {
    alphaBoolBlend {blendOption=Linear}
    boneToModify {boneName=dyn_zipper_base}
    constraintSetup [
     {targetBone {boneName=dyn_jacket_front_l_04}}
     {targetBone {boneName=dyn_jacket_front_l_05}}
     {targetBone {boneName=dyn_jacket_front_l_03}}
     {targetBone {boneName=dyn_jacket_front_r_04}}
    ]
    constraintWeights [1,1,1,1]
   }
   n12 = "Constraint - dyn_zipper_tab_MSTR (P)"!(ComponentPose=@n11, Alpha=1, ConstraintWeights_0=1, ConstraintWeights_1:float=0) -> (Pose) {
    alphaBoolBlend {blendOption=Linear}
    boneToModify {boneName=dyn_zipper_tab_MSTR}
    constraintSetup [{targetBone {boneName=dyn_zipper_base}}]
    constraintWeights [1,0]
   }
   n13 = Anim Dynamics - Bone: dyn_zipper_tab(ComponentPose=@n12, Alpha=1, ExternalForce="0,0,0") {
    angularDampingOverride=0.6
    angularSpringConstant=50
    bAngularSpring=true
    bOverrideAngularDamping=true
    bUseGravityOverride=true
    simulationSpace=RootRelative
    alphaBoolBlend {blendOption=Linear}
    boundBone {boneName=dyn_zipper_tab}
    componentLinearAccScale=<1 1 1>
    componentLinearVelScale=<1 1 1>
    gravityOverride {z=-400}
    physicsBodyDefinitions [{
      boundBone {boneName=dyn_zipper_tab}
      boxExtents=<5 5 5>
      constraintSetup {angularLimitsMax=<5 0 10>, angularLimitsMin=<-5 -80 -10>, angularTarget=<1 0 0>}
      localJointOffset=<0.75 0 0>
     }]
   }
   n14 = "Transform (Modify) Bone - Bone: dyn_jacket_front_r_MSTR"!(ComponentPose=@n13, bAlphaBoolEnabled:bool=True, Translation="0,0,-5", Rotation="10,0,0", Scale="1,1,1") -> (Pose) {
    alphaInputType=Bool
    rotationMode=BMM_Additive
    rotationSpace=BCS_BoneSpace
    translationMode=BMM_Additive
    translationSpace=BCS_BoneSpace
    alphaBoolBlend {blendInTime=1, blendOption=Linear, blendOutTime=1}
    boneToModify {boneName=dyn_jacket_front_r_MSTR}
    rotation {pitch=10}
    translation {z=-5}
   }
   n15 = "Transform (Modify) Bone - Bone: dyn_jacket_side_r_MSTR"!(ComponentPose=@n14, bAlphaBoolEnabled:bool=True, Translation="0,0,-5", Rotation="10,0,0", Scale="1,1,1") -> (Pose) {
    alphaInputType=Bool
    rotationMode=BMM_Additive
    rotationSpace=BCS_BoneSpace
    translationMode=BMM_Additive
    translationSpace=BCS_BoneSpace
    alphaBoolBlend {blendInTime=1, blendOption=Linear, blendOutTime=1}
    boneToModify {boneName=dyn_jacket_side_r_MSTR}
    rotation {pitch=10}
    translation {z=-5}
   }
   n16 = "Transform (Modify) Bone - Bone: dyn_jacket_front_l_MSTR"!(ComponentPose=@n15, bAlphaBoolEnabled:bool=True, Translation="0,0,2", Rotation="5,0,0", Scale="1,1,1") -> (Pose) {
    alphaInputType=Bool
    rotationMode=BMM_Additive
    rotationSpace=BCS_BoneSpace
    translationMode=BMM_Additive
    translationSpace=BCS_BoneSpace
    alphaBoolBlend {blendInTime=1, blendOption=Linear, blendOutTime=1}
    boneToModify {boneName=dyn_jacket_front_l_MSTR}
    rotation {pitch=5}
    translation {z=2}
   }
   n17 = "Transform (Modify) Bone - Bone: dyn_jacket_side_l_MSTR"!(ComponentPose=@n16, bAlphaBoolEnabled:bool=True, Translation="0,0,2", Rotation="5,0,0", Scale="1,1,1") -> (Pose) {
    alphaInputType=Bool
    rotationMode=BMM_Additive
    rotationSpace=BCS_BoneSpace
    translationMode=BMM_Additive
    translationSpace=BCS_BoneSpace
    alphaBoolBlend {blendInTime=1, blendOption=Linear, blendOutTime=1}
    boneToModify {boneName=dyn_jacket_side_l_MSTR}
    rotation {pitch=5}
    translation {z=2}
   }
   n18 = "Transform (Modify) Bone - Bone: dyn_jacket_rear_MSTR"!(ComponentPose=@n17, bAlphaBoolEnabled:bool=True, Translation="0,0,0", Rotation="-5,0,0", Scale="1,1,1") -> (Pose) {
    alphaInputType=Bool
    rotationMode=BMM_Additive
    rotationSpace=BCS_BoneSpace
    alphaBoolBlend {blendOption=Linear}
    boneToModify {boneName=dyn_jacket_rear_MSTR}
    rotation {pitch=-5}
   }
   n19 = "Transform (Modify) Bone - Bone: dyn_jacket_rear_l_MSTR"!(ComponentPose=@n18, bAlphaBoolEnabled:bool=True, Translation="0,0,0", Rotation="-10,0,0", Scale="1,1,1") -> (Pose) {
    alphaInputType=Bool
    rotationMode=BMM_Additive
    rotationSpace=BCS_BoneSpace
    alphaBoolBlend {blendOption=Linear}
    boneToModify {boneName=dyn_jacket_rear_l_MSTR}
    rotation {pitch=-10}
   }
   n20 = "Transform (Modify) Bone - Bone: dyn_jacket_rear_r_MSTR"!(ComponentPose=@n19, bAlphaBoolEnabled:bool=True, Translation="0,0,0", Rotation="-10,0,0", Scale="1,1,1") -> (Pose) {
    alphaInputType=Bool
    rotationMode=BMM_Additive
    rotationSpace=BCS_BoneSpace
    alphaBoolBlend {blendOption=Linear}
    boneToModify {boneName=dyn_jacket_rear_r_MSTR}
    rotation {pitch=-10}
   }
   n21 = "Transform (Modify) Bone - Bone: dyn_jacket_side_r_MSTR"(ComponentPose=@n20, bAlphaBoolEnabled=True, Translation="0,0,0", Rotation="-9.530,-10.475,1.676", Scale="1,1,1") {
    alphaInputType=Bool
    rotationMode=BMM_Additive
    rotationSpace=BCS_BoneSpace
    alphaBoolBlend {blendOption=Linear}
    boneToModify {boneName=dyn_jacket_side_r_MSTR}
    rotation=<rot -9.53 -10.4749 1.676>
   }
   Output Pose(Result=Component To Local(ComponentPose=@n21)) {layerGroup=None}
   n22 = Control Rig!(
    Source=Copy Pose From Mesh() {bUseAttachedParent=true}
    Alpha
    PantsHarmonicsSpeed:double=Get Pants Harmonics Speed() -> (Pants Harmonics Speed:double)
    Parachuting:double=Get ParachutingAlpha() -> (ParachutingAlpha:double)
    Skydiving:double=Get SkydiveAlpha() -> (SkydiveAlpha:double)
    SprintJump:double=Get JacketJumpAlpha() -> (JacketJumpAlpha:double)
   ) -> (Pose) {
    alphaBoolBlend {blendOption=Linear}
    controlRigAssetReference {blueprintRigClass=&/M_MED_DryEraseToro_CtrlRig.M_MED_DryEraseToro_CtrlRig_C}
   }
   Save cached pose 'CR Cache'(Pose=@n22) {cacheName=CR Cache}
  ]}
 EventGraph {nodes [
   Event BlueprintUpdateAnimation()
   > n1 = Sequence!(execute) -> (then_0, then_1, then_2:Exec, then_3:Exec)
   > n2 = "Gravity Override State Function (Parameter Asset) (Nativized)"(
    CharacterPartAnimInstance=Self-Reference()
    GameTimeInSeconds=GetGameTimeInSeconds()
    Parameters=&/M_MED_DryEraseToro_Parameters
   )
   > Set Gravity Override(Gravity Override=@n2)
   > Set RBAN Settings Out(RBAN Settings Out=@n2.OutRigidBodyAnimNodeParameters)
   > n3 = Branch(Condition=AND Boolean(A=Get bIsJumping(), B=Get bIsTacticalSprinting()))
   > Set JacketJumpAlpha(JacketJumpAlpha:double=1) -> (Output_Get:double)
   Set JacketJumpAlpha(execute=@n3.else, JacketJumpAlpha=0)
   Branch(execute=@n1.then_2, Condition=Get bRigidBodyEnabled())
   > n4 = Branch(Condition=Get bIsParachuteOpened())
   > Set ParachutingAlpha(ParachutingAlpha:double=1) -> (Output_Get:double)
   Set ParachutingAlpha(execute=@n4.else, ParachutingAlpha=0)
   n5 = Branch(execute=@n1.then_3, Condition=Get bIsInFrontEnd())
   > Set RBAN_Alpha(RBAN_Alpha:double=0.5) -> (Output_Get:double)
   Set RBAN_Alpha(execute=@n5.else, RBAN_Alpha=1)
   n6 = Branch(execute=@n1.then_1, Condition=Get bRigidBodyEnabled())
   > n7 = Branch(Condition=Get bIsSkyDiving())
   > Set SkydiveAlpha(SkydiveAlpha:double=1) -> (Output_Get:double)
   > n8 = Branch(Condition=Get bIsSkydiveFloating())
   > Set Pants Harmonics Speed(Pants Harmonics Speed:double=1.75) -> (Output_Get:double)
   n9 = Branch(execute=@n8.else, Condition=Get bIsSkydiveDiveMode())
   > Set Pants Harmonics Speed(Pants Harmonics Speed=2.25)
   Set Pants Harmonics Speed(execute=@n9.else, Pants Harmonics Speed=1.5)
   Set SkydiveAlpha(execute=@n7.else, SkydiveAlpha=0)
   > Set Pants Harmonics Speed(Pants Harmonics Speed=MapRangeClamped(Value=Get PawnSpeed2D(), InRangeA=600, InRangeB=1500))
   Set Pants Harmonics Speed(execute=@n6.else, Pants Harmonics Speed=0)
   Event OnCharacterCustomizationLoaded()
   n10 = CheckForBackbling()
   > Set isWearingPack(isWearingPack:bool=OR Boolean!(A=@n10, B=@n10.bIsBackblingACape, C:bool=@n10.bIsBackblingLong) -> (ReturnValue)) -> (Output_Get:bool)
   > Set isWearingCape(isWearingCape=@n10.bIsBackblingLong)
  ]}
]
variables [
 Gravity Override {type=Vector}
 isWearingCape {type=boolean}
 isWearingPack {type=boolean}
 JacketJumpAlpha {type=number}
 Pants Harmonics Speed {type=number}
 ParachutingAlpha {type=number}
 RBAN Settings Out {type=FortRigidBodyAnimNodeParameters}
 RBAN_Alpha {type=number}
 SkydiveAlpha {type=number}
]
