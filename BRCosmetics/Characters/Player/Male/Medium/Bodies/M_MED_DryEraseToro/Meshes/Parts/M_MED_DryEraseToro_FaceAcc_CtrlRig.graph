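# Control Rig for the DryEraseToro face accessory that re-poses hat bones, including the rear and right side-back, as the head turns.
# A spherical pose reader measures the head against spine_05, and its smoothed output drives a Modify Transforms weight.
# That weight is multiplied by a smoothed TargetingAlpha, so the correction fades in with targeting.
# Only this first chain runs from Begin Execution; other pose readers, fixed-weight transforms and a branch sit outside it.
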ControlRigBlueprint /BRCosmetics/Characters/Player/Male/Medium/Bodies/M_MED_DryEraseToro/Meshes/Parts/M_MED_DryEraseToro_FaceAcc_CtrlRig
graphs [
 RigVMFunctionLibrary {links [], nodes []}
 RigVMModel {nodes [
   RigUnit_BeginExecution = &RigUnit_BeginExecution()
   n1 = &RigUnit_SphericalPoseReader(
    ActiveRegionScaleFactors=(PositiveWidth=1,NegativeWidth=1,PositiveHeight=0.5,NegativeHeight=1)
    ActiveRegionSize=0.1
    Debug=(bDrawDebug=true,bDraw2D=False,bDrawLocalAxes=False,DebugScale=25,DebugSegments=20,DebugThickness=0.25)
    DriverAxis=(X=0,Y=1,Z=0)
    DriverItem=(Type=Bone,Name="head")
    FalloffRegionScaleFactors=(PositiveWidth=1,NegativeWidth=1,PositiveHeight=1,NegativeHeight=1)
    FalloffSize=0.3
    FlipHeightScaling=False
    FlipWidthScaling=False
    OptionalParentItem=(Type=Bone,Name="spine_05")
    RotationOffset=(X=0,Y=-90,Z=-10)
    ExecutePin=@RigUnit_BeginExecution.ExecutePin
   )
   AlphaInterp = &RigVMFunction_AlphaInterp(InterpSpeedDecreasing=4, InterpSpeedIncreasing=8, Scale=1, bClampResult=true, Value=RigVMVariableNode() {variable=TargetingAlpha}.Value)
   RigVMFunction_AlphaInterp = &RigVMFunction_AlphaInterp(InterpSpeedDecreasing=20, InterpSpeedIncreasing=20, Scale=1, bClampResult=False, bInterpResult=True, Value=@n1.OutputParam)
   > &RigUnit_ModifyTransforms(
    ItemToModify="((Item=(Type=Bone,Name=\"dyn_hat_sideback_r_MSTR\"),Transform=(Rotation=(X=0,Y=-0.8192,Z=0,W=0.5736),Translation=(X=0,Y=0,Z=0),Scale3D=(X=1,Y=1,Z=1))),(Item=(Type=Bone,Name=\"dyn_hat_rear_MSTR\"),Transform=(Rotation=(X=-0.1485,Y=-0.7912,Z=-0.212,W=0.554),Trans ...(truncated)"
    WeightMinimum=0
    Weight=&RigVMFunction_MathDoubleMul(B=@AlphaInterp.Result, A=@RigVMFunction_AlphaInterp.Result).Result
   )
   n2 = &RigUnit_SphericalPoseReader(
    ActiveRegionScaleFactors=(PositiveWidth=1,NegativeWidth=1,PositiveHeight=0.1,NegativeHeight=1)
    ActiveRegionSize=0.05
    Debug=(bDrawDebug=true,bDraw2D=False,bDrawLocalAxes=False,DebugScale=25,DebugSegments=20,DebugThickness=0.25)
    DriverAxis=(X=0,Y=1,Z=0)
    DriverItem=(Type=Bone,Name="head")
    ExecutePin=()
    FalloffRegionScaleFactors=(PositiveWidth=1,NegativeWidth=1,PositiveHeight=1,NegativeHeight=1)
    FalloffSize=0.1
    FlipHeightScaling=False
    FlipWidthScaling=False
    OptionalParentItem=(Type=Bone,Name="spine_05")
    RotationOffset=(X=0,Y=-95,Z=35)
   )
   n3 = &RigVMFunction_AlphaInterp(InterpSpeedDecreasing=5, InterpSpeedIncreasing=5, Scale=1, bClampResult=False, bInterpResult=True, Value=@n2.OutputParam)
   > &RigUnit_ModifyTransforms(
    ItemToModify="((Item=(Type=Bone,Name=\"dyn_hat_sideback_r_MSTR\"),Transform=(Rotation=(X=0,Y=-0.766,Z=0,W=0.6428),Translation=(X=0,Y=0,Z=0),Scale3D=(X=1,Y=1,Z=1))),(Item=(Type=Bone,Name=\"dyn_hat_rear_MSTR\"),Transform=(Rotation=(X=-0.3214,Y=-0.3214,Z=-0.117,W=0.883),Transl ...(truncated)"
    WeightMinimum=0
    Weight=&RigVMFunction_MathDoubleMul(B=@AlphaInterp.Result, A=@n3.Result).Result
   )
   &RigUnit_ModifyTransforms(
    ExecutePin=()
    ItemToModify="((Item=(Type=Bone,Name=\"dyn_hat_rear_MSTR\"),Transform=(Rotation=(X=-0.0996,Y=-0.8067,Z=-0.1422,W=0.5649),Translation=(X=0,Y=0,Z=0),Scale3D=(X=1,Y=1,Z=1))),(Item=(Type=Bone,Name=\"dyn_hat_sideback_r_MSTR\"),Transform=(Rotation=(X=-0.1422,Y=-0.8067,Z=-0.0996,W ...(truncated)"
    Weight=1
    WeightMinimum=0
   )
   &RigUnit_ModifyTransforms(
    ExecutePin=()
    ItemToModify="((Item=(Type=Bone,Name=\"dyn_hat_sideback_r_MSTR\"),Transform=(Rotation=(X=0,Y=-0.766,Z=0,W=0.6428),Translation=(X=0,Y=0,Z=0),Scale3D=(X=1,Y=1,Z=1))),(Item=(Type=Bone,Name=\"dyn_hat_rear_MSTR\"),Transform=(Rotation=(X=-0.2241,Y=-0.483,Z=-0.1294,W=0.8365),Trans ...(truncated)"
    Weight=1
    WeightMinimum=0
   )
   n4 = &RigUnit_SphericalPoseReader(
    ActiveRegionScaleFactors=(PositiveWidth=1,NegativeWidth=1,PositiveHeight=0.25,NegativeHeight=0.5)
    ActiveRegionSize=0.15
    Debug=(bDrawDebug=true,bDraw2D=False,bDrawLocalAxes=False,DebugScale=25,DebugSegments=20,DebugThickness=0.25)
    DriverAxis=(X=0,Y=1,Z=0)
    DriverItem=(Type=Bone,Name="head")
    ExecutePin=()
    FalloffRegionScaleFactors=(PositiveWidth=1,NegativeWidth=1,PositiveHeight=0.25,NegativeHeight=0.5)
    FalloffSize=0.3
    FlipHeightScaling=False
    FlipWidthScaling=False
    OptionalParentItem=(Type=Bone,Name="spine_05")
    RotationOffset=(X=34.537,Y=-71.8582,Z=6.4292)
   )
   n5 = &RigVMFunction_AlphaInterp(InterpSpeedDecreasing=10, InterpSpeedIncreasing=10, Scale=1, bClampResult=False, bInterpResult=True, Value=@n4.OutputParam)
   Multiply = &RigVMFunction_MathDoubleMul(B=1, A=@n5.Result)
   RigVMFunction_ControlFlowBranch = &RigVMFunction_ControlFlowBranch(Condition=true, ExecuteContext=())
  ]}
]
hierarchy [
 root {loc=<0 0 0>, rot=<0 0 0 1>}
 pelvis {parent=root, loc=<0 -0.7071 93.1191>, rot=<0.0127 0.707 -0.0127 -0.707>}
 spine_01 {parent=pelvis, loc=<6.7494 -0.6388 0>, rot=<0 0 0.034 -0.9994>}
 spine_02 {parent=spine_01, loc=<7.2518 -0.035 0>, rot=<0 0 0.0238 -0.9997>}
 spine_03 {parent=spine_02, loc=<6.0104 0 0>, rot=<0 0 0.0149 -0.9999>}
 spine_04 {parent=spine_03, loc=<8.1024 0 0>, rot=<0 0 -0.0132 -0.9999>}
 spine_05 {parent=spine_04, loc=<10.0146 0 0>, rot=<0 0 -0.0015 -1>}
 neck_01 {parent=spine_05, loc=<11.3138 0.1529 0>, rot=<0 0 -0.2003 -0.9797>}
 neck_02 {parent=neck_01, loc=<4.5627 0 0>, rot=<0 0 0.0504 -0.9987>}
 head {parent=neck_02, loc=<5.3914 0 0>, rot=<0 0 0.115 -0.9934>}
 dyn_hat_front_MSTR {parent=head, loc=<10.9136 8.6011 -0.0073>, rot=<0.3148 0.6327 0.6339 0.3142>}
 dyn_hat_front_01 {parent=dyn_hat_front_MSTR, loc=<0 0 0>, rot=<0 0 0 1>}
 dyn_hat_front_02 {parent=dyn_hat_front_01, loc=<3.875 0 0>, rot=<0 0 0 1>}
 dyn_hat_frontside_l_MSTR {parent=head, loc=<10.364 5.3497 -5.9273>, rot=<-0.2385 -0.853 -0.3096 -0.3459>}
 dyn_hat_frontside_l_01 {parent=dyn_hat_frontside_l_MSTR, loc=<0 0 0>, rot=<0 0 0 1>}
 dyn_hat_frontside_l_02 {parent=dyn_hat_frontside_l_01, loc=<4.3939 0 0>, rot=<0 0 0 1>}
 dyn_hat_side_l_MSTR {parent=head, loc=<8.5808 -1.313 -8.0896>, rot=<-0.1673 -0.9295 -0.0095 -0.3286>}
 dyn_hat_side_l_01 {parent=dyn_hat_side_l_MSTR, loc=<0 0 0>, rot=<0 0 0 1>}
 dyn_hat_side_l_02 {parent=dyn_hat_side_l_01, loc=<4.0999 0 0>, rot=<0 0 0 1>}
 dyn_hat_sideback_l_MSTR {parent=head, loc=<7.1096 -7.4553 -6.1835>, rot=<0.0092 -0.9044 0.3516 -0.2417>}
 dyn_hat_sideback_l_01 {parent=dyn_hat_sideback_l_MSTR, loc=<0 0 0>, rot=<0 0 0 1>}
 dyn_hat_sideback_l_02 {parent=dyn_hat_sideback_l_01, loc=<3.9325 0 0>, rot=<0 0 0 1>}
 dyn_hat_rear_MSTR {parent=head, loc=<6.3425 -9.9018 0.0219>, rot=<0.1659 -0.6903 0.684 -0.1674>}
 dyn_hat_rear_l_01 {parent=dyn_hat_rear_MSTR, loc=<0 0 0>, rot=<0 0 0 1>}
 dyn_hat_rear_l_02 {parent=dyn_hat_rear_l_01, loc=<4.0534 0 0>, rot=<0 0 0 1>}
 dyn_hat_frontside_r_MSTR {parent=head, loc=<10.3639 5.3497 5.9273>, rot=<0.853 -0.2385 -0.3459 0.3096>}
 dyn_hat_frontside_r_01 {parent=dyn_hat_frontside_r_MSTR, loc=<0 0 0>, rot=<0 0 0 1>}
 dyn_hat_frontside_r_02 {parent=dyn_hat_frontside_r_01, loc=<-4.2714 0 -0.0001>, rot=<0 0 0 1>}
 dyn_hat_side_r_MSTR {parent=head, loc=<8.5809 -1.313 8.0896>, rot=<0.9295 -0.1673 -0.3286 0.0095>}
 dyn_hat_side_r_01 {parent=dyn_hat_side_r_MSTR, loc=<0 0 0>, rot=<0 0 0 1>}
 dyn_hat_side_r_02 {parent=dyn_hat_side_r_01, loc=<-3.9773 -0.0001 -0.0001>, rot=<0 0 0 1>}
 dyn_hat_sideback_r_MSTR {parent=head, loc=<7.1096 -7.4553 6.1835>, rot=<0.9044 0.0092 -0.2417 -0.3516>}
 dyn_hat_sideback_r_01 {parent=dyn_hat_sideback_r_MSTR, loc=<0 0 0>, rot=<0 0 0 1>}
 dyn_hat_sideback_r_02 {parent=dyn_hat_sideback_r_01, loc=<-3.8101 0 0>, rot=<0 0 0 1>}
 transform2 {parent=spine_05, loc=<-131.0232 -6.9525 0>, rot=<-0.0284 0.7065 -0.0284 -0.7065>}
 clavicle_l {parent=spine_05, loc=<4.3575 -0.1466 -2.3223>, rot=<0.0465 -0.7289 -0.009 -0.6829>}
 upperarm_l {parent=clavicle_l, loc=<15.5111 0 0>, rot=<0.0226 -0.3234 -0.0088 -0.9459>}
 clavicle_r {parent=spine_05, loc=<4.3575 -0.1466 2.3223>, rot=<0.7289 0.0465 -0.6829 0.009>}
 upperarm_r {parent=clavicle_r, loc=<-15.5111 0 0>, rot=<0.0226 -0.3234 -0.0088 -0.9459>}
]
variables [TargetingAlpha {type=double, default=0}]
