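# Animation Blueprint for the DryEraseToro face accessory, adding secondary motion to its hat bones.
# It copies poses from the attached parent and CharacterMesh (set on customization load), layered at both upper arms, into a Control Rig whose TargetingAlpha switches between 0 and 1 with bIsTargeting.
# A serial chain of Anim Dynamics nodes then swings two-bone hat segments around the brim, using damped angular springs, gravity and tight angular limits.
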AnimBlueprint /BRCosmetics/Characters/Player/Male/Medium/Bodies/M_MED_DryEraseToro/Meshes/Parts/M_MED_DryEraseToro_FaceAcc_AnimBP
target_skeleton=&/M_MED_DryEraseToro_FaceAcc_Skeleton
graphs [
 AnimGraph {nodes [
   n1 = Get CharacterMesh() -> (CharacterMesh:Skeletal Mesh Component Object Reference)
   n2 = Control Rig!(
    Source=Layered blend per bone(
     BasePose=Copy Pose From Mesh() {bUseAttachedParent=true}
     BlendPoses_0=Copy Pose From Mesh(SourceMeshComponent=@n1) {bUseAttachedParent=true}
    ) {layerSetup [{branchFilters [{boneName=upperarm_l}, {boneName=upperarm_r}]}]}
    Alpha
    TargetingAlpha:double=Get TargetingAlpha() -> (TargetingAlpha:double)
   ) -> (Pose) {
    alphaBoolBlend {blendOption=Linear}
    controlRigAssetReference {blueprintRigClass=&/M_MED_DryEraseToro_FaceAcc_CtrlRig.M_MED_DryEraseToro_FaceAcc_CtrlRig_C}
   }
   n3 = Anim Dynamics - Bone: dyn_hat_front_01(ComponentPose=Local To Component(LocalPose=@n2), Alpha=1, ExternalForce="0,0,0") {
    angularDampingOverride=0.999
    angularSpringConstant=10
    bAngularSpring=true
    bGravityOverrideInSimSpace=true
    bOverrideAngularDamping=true
    bUseGravityOverride=true
    simulationSpace=RootRelative
    alphaBoolBlend {blendOption=Linear}
    boundBone {boneName=dyn_hat_front_01}
    componentAppliedLinearAccClamp=<5000 5000 5000>
    componentLinearAccScale=<1 1 1>
    componentLinearVelScale=<1 1 1>
    gravityOverride {z=-980}
    physicsBodyDefinitions [{
      sphereCollisionRadius=0
      boundBone {boneName=dyn_hat_front_01}
      boxExtents=<30 30 30>
      constraintSetup {angularLimitsMax=<1 0 2>, angularLimitsMin=<-1 -20 -2>, angularTarget=<1 0 0>}
      localJointOffset=<1.5 0 0>
     }]
    relativeSpaceBone {boneName=head}
   }
   n4 = Anim Dynamics - Bone: dyn_hat_front_02(ComponentPose=@n3, Alpha=1, ExternalForce="0,0,0") {
    angularDampingOverride=0.999
    angularSpringConstant=10
    bAngularSpring=true
    bGravityOverrideInSimSpace=true
    bOverrideAngularDamping=true
    bUseGravityOverride=true
    simulationSpace=RootRelative
    alphaBoolBlend {blendOption=Linear}
    boundBone {boneName=dyn_hat_front_02}
    componentAppliedLinearAccClamp=<5000 5000 5000>
    componentLinearAccScale=<1 1 1>
    componentLinearVelScale=<1 1 1>
    gravityOverride {z=-980}
    physicsBodyDefinitions [{
      sphereCollisionRadius=0
      boundBone {boneName=dyn_hat_front_02}
      boxExtents=<30 30 30>
      constraintSetup {angularLimitsMax=<1 0 2>, angularLimitsMin=<-1 -20 -2>, angularTarget=<1 0 0>}
      localJointOffset=<1.5 0 0>
     }]
    relativeSpaceBone {boneName=head}
   }
   n5 = Anim Dynamics - Bone: dyn_hat_frontside_l_01(ComponentPose=@n4, Alpha=1, ExternalForce="0,0,0") {
    angularDampingOverride=0.999
    angularSpringConstant=10
    bAngularSpring=true
    bGravityOverrideInSimSpace=true
    bOverrideAngularDamping=true
    bUseGravityOverride=true
    simulationSpace=RootRelative
    alphaBoolBlend {blendOption=Linear}
    boundBone {boneName=dyn_hat_frontside_l_01}
    componentAppliedLinearAccClamp=<5000 5000 5000>
    componentLinearAccScale=<1 1 1>
    componentLinearVelScale=<1 1 1>
    gravityOverride {z=-980}
    physicsBodyDefinitions [{
      sphereCollisionRadius=0
      boundBone {boneName=dyn_hat_frontside_l_01}
      boxExtents=<30 30 30>
      constraintSetup {angularLimitsMax=<1 0 2>, angularLimitsMin=<-1 -20 -2>, angularTarget=<1 0 0>}
      localJointOffset=<1.5 0 0>
     }]
    relativeSpaceBone {boneName=head}
   }
   n6 = Anim Dynamics - Bone: dyn_hat_frontside_l_02(ComponentPose=@n5, Alpha=1, ExternalForce="0,0,0") {
    angularDampingOverride=0.999
    angularSpringConstant=10
    bAngularSpring=true
    bGravityOverrideInSimSpace=true
    bOverrideAngularDamping=true
    bUseGravityOverride=true
    simulationSpace=RootRelative
    alphaBoolBlend {blendOption=Linear}
    boundBone {boneName=dyn_hat_frontside_l_02}
    componentAppliedLinearAccClamp=<5000 5000 5000>
    componentLinearAccScale=<1 1 1>
    componentLinearVelScale=<1 1 1>
    gravityOverride {z=-980}
    physicsBodyDefinitions [{
      sphereCollisionRadius=0
      boundBone {boneName=dyn_hat_frontside_l_02}
      boxExtents=<30 30 30>
      constraintSetup {angularLimitsMax=<1 0 2>, angularLimitsMin=<-1 -20 -2>, angularTarget=<1 0 0>}
      localJointOffset=<1.5 0 0>
     }]
    relativeSpaceBone {boneName=head}
   }
   n7 = Anim Dynamics - Bone: dyn_hat_side_l_01(ComponentPose=@n6, Alpha=1, ExternalForce="0,0,0") {
    angularDampingOverride=0.999
    angularSpringConstant=10
    bAngularSpring=true
    bGravityOverrideInSimSpace=true
    bOverrideAngularDamping=true
    bUseGravityOverride=true
    simulationSpace=RootRelative
    alphaBoolBlend {blendOption=Linear}
    boundBone {boneName=dyn_hat_side_l_01}
    componentAppliedLinearAccClamp=<5000 5000 5000>
    componentLinearAccScale=<1 1 1>
    componentLinearVelScale=<1 1 1>
    gravityOverride {z=-980}
    physicsBodyDefinitions [{
      sphereCollisionRadius=0
      boundBone {boneName=dyn_hat_side_l_01}
      boxExtents=<30 30 30>
      constraintSetup {angularLimitsMax=<1 0 2>, angularLimitsMin=<-1 -20 -2>, angularTarget=<1 0 0>}
      localJointOffset=<1.5 0 0>
     }]
    relativeSpaceBone {boneName=head}
   }
   n8 = Anim Dynamics - Bone: dyn_hat_side_l_02(ComponentPose=@n7, Alpha=1, ExternalForce="0,0,0") {
    angularDampingOverride=0.999
    angularSpringConstant=10
    bAngularSpring=true
    bGravityOverrideInSimSpace=true
    bOverrideAngularDamping=true
    bUseGravityOverride=true
    simulationSpace=RootRelative
    alphaBoolBlend {blendOption=Linear}
    boundBone {boneName=dyn_hat_side_l_02}
    componentAppliedLinearAccClamp=<5000 5000 5000>
    componentLinearAccScale=<1 1 1>
    componentLinearVelScale=<1 1 1>
    gravityOverride {z=-980}
    physicsBodyDefinitions [{
      sphereCollisionRadius=0
      boundBone {boneName=dyn_hat_side_l_02}
      boxExtents=<30 30 30>
      constraintSetup {angularLimitsMax=<1 0 2>, angularLimitsMin=<-1 -20 -2>, angularTarget=<1 0 0>}
      localJointOffset=<1.5 0 0>
     }]
    relativeSpaceBone {boneName=head}
   }
   n9 = Anim Dynamics - Bone: dyn_hat_sideback_l_01(ComponentPose=@n8, Alpha=1, ExternalForce="0,0,0") {
    angularDampingOverride=0.999
    angularSpringConstant=10
    bAngularSpring=true
    bGravityOverrideInSimSpace=true
    bOverrideAngularDamping=true
    bUseGravityOverride=true
    simulationSpace=RootRelative
    alphaBoolBlend {blendOption=Linear}
    boundBone {boneName=dyn_hat_sideback_l_01}
    componentAppliedLinearAccClamp=<5000 5000 5000>
    componentLinearAccScale=<1 1 1>
    componentLinearVelScale=<1 1 1>
    gravityOverride {z=-980}
    physicsBodyDefinitions [{
      sphereCollisionRadius=0
      boundBone {boneName=dyn_hat_sideback_l_01}
      boxExtents=<30 30 30>
      constraintSetup {angularLimitsMax=<1 0 2>, angularLimitsMin=<-1 -20 -2>, angularTarget=<1 0 0>}
      localJointOffset=<1.5 0 0>
     }]
    relativeSpaceBone {boneName=head}
   }
   n10 = Anim Dynamics - Bone: dyn_hat_sideback_l_02(ComponentPose=@n9, Alpha=1, ExternalForce="0,0,0") {
    angularDampingOverride=0.999
    angularSpringConstant=10
    bAngularSpring=true
    bGravityOverrideInSimSpace=true
    bOverrideAngularDamping=true
    bUseGravityOverride=true
    simulationSpace=RootRelative
    alphaBoolBlend {blendOption=Linear}
    boundBone {boneName=dyn_hat_sideback_l_02}
    componentAppliedLinearAccClamp=<5000 5000 5000>
    componentLinearAccScale=<1 1 1>
    componentLinearVelScale=<1 1 1>
    gravityOverride {z=-980}
    physicsBodyDefinitions [{
      sphereCollisionRadius=0
      boundBone {boneName=dyn_hat_sideback_l_02}
      boxExtents=<30 30 30>
      constraintSetup {angularLimitsMax=<1 0 2>, angularLimitsMin=<-1 -20 -2>, angularTarget=<1 0 0>}
      localJointOffset=<1.5 0 0>
     }]
    relativeSpaceBone {boneName=head}
   }
   n11 = Anim Dynamics - Bone: dyn_hat_rear_l_01(ComponentPose=@n10, Alpha=1, ExternalForce="0,0,0") {
    angularDampingOverride=0.999
    angularSpringConstant=10
    bAngularSpring=true
    bGravityOverrideInSimSpace=true
    bOverrideAngularDamping=true
    bUseGravityOverride=true
    simulationSpace=RootRelative
    alphaBoolBlend {blendOption=Linear}
    boundBone {boneName=dyn_hat_rear_l_01}
    componentAppliedLinearAccClamp=<5000 5000 5000>
    componentLinearAccScale=<1 1 1>
    componentLinearVelScale=<1 1 1>
    gravityOverride {z=-980}
    physicsBodyDefinitions [{
      sphereCollisionRadius=0
      boundBone {boneName=dyn_hat_rear_l_01}
      boxExtents=<30 30 30>
      constraintSetup {angularLimitsMax=<1 0 2>, angularLimitsMin=<-1 -20 -2>, angularTarget=<1 0 0>}
      localJointOffset=<1.5 0 0>
     }]
    relativeSpaceBone {boneName=head}
   }
   n12 = Anim Dynamics - Bone: dyn_hat_rear_l_02(ComponentPose=@n11, Alpha=1, ExternalForce="0,0,0") {
    angularDampingOverride=0.999
    angularSpringConstant=10
    bAngularSpring=true
    bGravityOverrideInSimSpace=true
    bOverrideAngularDamping=true
    bUseGravityOverride=true
    simulationSpace=RootRelative
    alphaBoolBlend {blendOption=Linear}
    boundBone {boneName=dyn_hat_rear_l_02}
    componentAppliedLinearAccClamp=<5000 5000 5000>
    componentLinearAccScale=<1 1 1>
    componentLinearVelScale=<1 1 1>
    gravityOverride {z=-980}
    physicsBodyDefinitions [{
      sphereCollisionRadius=0
      boundBone {boneName=dyn_hat_rear_l_02}
      boxExtents=<30 30 30>
      constraintSetup {angularLimitsMax=<1 0 2>, angularLimitsMin=<-1 -20 -2>, angularTarget=<1 0 0>}
      localJointOffset=<1.5 0 0>
     }]
    relativeSpaceBone {boneName=head}
   }
   n13 = Anim Dynamics - Bone: dyn_hat_sideback_r_01(ComponentPose=@n12, Alpha=1, ExternalForce="0,0,0") {
    angularDampingOverride=0.999
    angularSpringConstant=10
    bAngularSpring=true
    bGravityOverrideInSimSpace=true
    bOverrideAngularDamping=true
    bUseGravityOverride=true
    simulationSpace=RootRelative
    alphaBoolBlend {blendOption=Linear}
    boundBone {boneName=dyn_hat_sideback_r_01}
    componentAppliedLinearAccClamp=<5000 5000 5000>
    componentLinearAccScale=<1 1 1>
    componentLinearVelScale=<1 1 1>
    gravityOverride {z=-980}
    physicsBodyDefinitions [{
      sphereCollisionRadius=0
      boundBone {boneName=dyn_hat_sideback_r_01}
      boxExtents=<30 30 30>
      constraintSetup {angularLimitsMax=<1 0 2>, angularLimitsMin=<-1 -20 -2>, angularTarget=<1 0 0>}
      localJointOffset=<-1.5 0 0>
     }]
    relativeSpaceBone {boneName=head}
   }
   n14 = Anim Dynamics - Bone: dyn_hat_sideback_r_02(ComponentPose=@n13, Alpha=1, ExternalForce="0,0,0") {
    angularDampingOverride=0.999
    angularSpringConstant=10
    bAngularSpring=true
    bGravityOverrideInSimSpace=true
    bOverrideAngularDamping=true
    bUseGravityOverride=true
    simulationSpace=RootRelative
    alphaBoolBlend {blendOption=Linear}
    boundBone {boneName=dyn_hat_sideback_r_02}
    componentAppliedLinearAccClamp=<5000 5000 5000>
    componentLinearAccScale=<1 1 1>
    componentLinearVelScale=<1 1 1>
    gravityOverride {z=-980}
    physicsBodyDefinitions [{
      sphereCollisionRadius=0
      boundBone {boneName=dyn_hat_sideback_r_02}
      boxExtents=<30 30 30>
      constraintSetup {angularLimitsMax=<1 0 2>, angularLimitsMin=<-1 -20 -2>, angularTarget=<1 0 0>}
      localJointOffset=<-1.5 0 0>
     }]
    relativeSpaceBone {boneName=head}
   }
   n15 = Anim Dynamics - Bone: dyn_hat_side_r_01(ComponentPose=@n14, Alpha=1, ExternalForce="0,0,0") {
    angularDampingOverride=0.999
    angularSpringConstant=10
    bAngularSpring=true
    bGravityOverrideInSimSpace=true
    bOverrideAngularDamping=true
    bUseGravityOverride=true
    simulationSpace=RootRelative
    alphaBoolBlend {blendOption=Linear}
    boundBone {boneName=dyn_hat_side_r_01}
    componentAppliedLinearAccClamp=<5000 5000 5000>
    componentLinearAccScale=<1 1 1>
    componentLinearVelScale=<1 1 1>
    gravityOverride {z=-980}
    physicsBodyDefinitions [{
      sphereCollisionRadius=0
      boundBone {boneName=dyn_hat_side_r_01}
      boxExtents=<30 30 30>
      constraintSetup {angularLimitsMax=<1 0 2>, angularLimitsMin=<-1 -20 -2>, angularTarget=<1 0 0>}
      localJointOffset=<-1.5 0 0>
     }]
    relativeSpaceBone {boneName=head}
   }
   n16 = Anim Dynamics - Bone: dyn_hat_side_r_02(ComponentPose=@n15, Alpha=1, ExternalForce="0,0,0") {
    angularDampingOverride=0.999
    angularSpringConstant=10
    bAngularSpring=true
    bGravityOverrideInSimSpace=true
    bOverrideAngularDamping=true
    bUseGravityOverride=true
    simulationSpace=RootRelative
    alphaBoolBlend {blendOption=Linear}
    boundBone {boneName=dyn_hat_side_r_02}
    componentAppliedLinearAccClamp=<5000 5000 5000>
    componentLinearAccScale=<1 1 1>
    componentLinearVelScale=<1 1 1>
    gravityOverride {z=-980}
    physicsBodyDefinitions [{
      sphereCollisionRadius=0
      boundBone {boneName=dyn_hat_side_r_02}
      boxExtents=<30 30 30>
      constraintSetup {angularLimitsMax=<1 0 2>, angularLimitsMin=<-1 -20 -2>, angularTarget=<1 0 0>}
      localJointOffset=<-1.5 0 0>
     }]
    relativeSpaceBone {boneName=head}
   }
   n17 = Anim Dynamics - Bone: dyn_hat_frontside_r_01(ComponentPose=@n16, Alpha=1, ExternalForce="0,0,0") {
    angularDampingOverride=0.999
    angularSpringConstant=10
    bAngularSpring=true
    bGravityOverrideInSimSpace=true
    bOverrideAngularDamping=true
    bUseGravityOverride=true
    simulationSpace=RootRelative
    alphaBoolBlend {blendOption=Linear}
    boundBone {boneName=dyn_hat_frontside_r_01}
    componentAppliedLinearAccClamp=<5000 5000 5000>
    componentLinearAccScale=<1 1 1>
    componentLinearVelScale=<1 1 1>
    gravityOverride {z=-980}
    physicsBodyDefinitions [{
      sphereCollisionRadius=0
      boundBone {boneName=dyn_hat_frontside_r_01}
      boxExtents=<30 30 30>
      constraintSetup {angularLimitsMax=<1 0 2>, angularLimitsMin=<-1 -20 -2>, angularTarget=<1 0 0>}
      localJointOffset=<-1.5 0 0>
     }]
    relativeSpaceBone {boneName=head}
   }
   n18 = Anim Dynamics - Bone: dyn_hat_frontside_r_02(ComponentPose=@n17, Alpha=1, ExternalForce="0,0,0") {
    angularDampingOverride=0.999
    angularSpringConstant=10
    bAngularSpring=true
    bGravityOverrideInSimSpace=true
    bOverrideAngularDamping=true
    bUseGravityOverride=true
    simulationSpace=RootRelative
    alphaBoolBlend {blendOption=Linear}
    boundBone {boneName=dyn_hat_frontside_r_02}
    componentAppliedLinearAccClamp=<5000 5000 5000>
    componentLinearAccScale=<1 1 1>
    componentLinearVelScale=<1 1 1>
    gravityOverride {z=-980}
    physicsBodyDefinitions [{
      sphereCollisionRadius=0
      boundBone {boneName=dyn_hat_frontside_r_02}
      boxExtents=<30 30 30>
      constraintSetup {angularLimitsMax=<1 0 2>, angularLimitsMin=<-1 -20 -2>, angularTarget=<1 0 0>}
      localJointOffset=<-1.5 0 0>
     }]
    relativeSpaceBone {boneName=head}
   }
   Output Pose(Result=Component To Local(ComponentPose=@n18)) {layerGroup=None}
  ]}
 EventGraph {nodes [
   Event BlueprintUpdateAnimation()
   > n1 = Branch(Condition=Get bIsTargeting())
   > Set TargetingAlpha(TargetingAlpha:double=1) -> (Output_Get:double)
   Set TargetingAlpha(execute=@n1.else, TargetingAlpha=0)
   n2 = Event OnCharacterCustomizationLoaded()
   > Set CharacterMesh(CharacterMesh:Skeletal Mesh Component Object Reference=@n2.CustomBaseMesh) -> (Output_Get:Skeletal Mesh Component Object Reference)
  ]}
]
variables [
 CharacterMesh {type=SkeletalMeshComponent}
 TargetingAlpha {type=number}
]
